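# Skydive physics asset for the DistantEchoPro male medium body cloak, simulating the cape and a dangling lightsaber.
# Capsule bodies cover spine, clavicles, arms and legs with free-moving joints, forming a body volume for the dynamic parts.
# Four seven-link cape chains hang from master bones on the clavicles and upper spine, with limited swing and zero-degree twist limits.
# A short two-link lightsaber chain hangs from a pelvis master bone with tight swing limits.
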PhysicsAsset /BRCosmetics/Characters/Player/Male/Medium/Bodies/M_MED_DistantEchoPro_Cloak/Meshes/M_MED_DistantEchoPro_Cloak_skydive_Physics1
bodies {
 calf_l [{shapeType=Capsule, length=30.3147, radius=13.0869, shapeName=calf_l_capsule, center=<-21.1879 0 0>, rotation=<rot 90 0 0>}]
 calf_r [{shapeType=Capsule, length=30.3147, radius=13.0869, shapeName=calf_l_capsule, center=<21.1879 0 0>, rotation=<rot 90 0 180>}]
 cape_01_l [{shapeType=Capsule, length=10.5974, radius=8.763, shapeName=cape_01_l_capsule, center=<9.0604 0 6.3148>, rotation=<rot 90 0 0>}]
 cape_01_r [
  {shapeType=Capsule, length=13.1635, radius=6.1968, shapeName=cape_01_r_capsule, center=<9.0604 0 2.8>, rotation=<rot 90 0 0>}
  {shapeType=Capsule, length=17.7358, radius=5.7925, shapeName=cape_01_r_capsule, center=<11.9455 0 1.6149>, rotation=<rot 0 0 90>}
 ]
 cape_02_l [{shapeType=Capsule, length=18.1446, radius=9.6584, shapeName=cape_02_l_capsule, center=<9.0604 0 6.3148>, rotation=<rot 70 89.9999 -0.0001>}]
 cape_02_r [
  {
   shapeType=Capsule
   length=19.6078
   radius=5.8793
   shapeName=cape_02_r_capsule
   center=<9.0604 0.4164 1.8>
   rotation=<rot 84.7003 -90.0001 -179.9998>
  }
  {shapeType=Capsule, length=22.9845, radius=7.0155, shapeName=cape_02_r_capsule, center=<3.6286 0 2.0556>, rotation=<rot 90 0 0>}
 ]
 cape_03_l [{shapeType=Capsule, length=16.0735, radius=9.6584, shapeName=cape_03_l_capsule, center=<9.0604 0 6.3148>, rotation=<rot 80 89.9998 -0.0002>}]
 cape_03_r [{shapeType=Capsule, length=18.8322, radius=9.6584, shapeName=cape_03_r_capsule, center=<9.0604 0 6.3>, rotation=<rot 90 0 0>}]
 cape_04_l [{shapeType=Capsule, length=14.2326, radius=9.6584, shapeName=cape_04_l_capsule, center=<9.0604 0 6.3148>, rotation=<rot 0 0 90>}]
 cape_04_r [{shapeType=Capsule, length=14.2326, radius=9.6584, shapeName=cape_04_r_capsule, center=<9.0604 0 6.3>, rotation=<rot 0 0 90>}]
 cape_05_l [{shapeType=Capsule, length=14.3015, radius=10.5715, shapeName=cape_05_l_capsule, center=<9.0604 0.4838 3.5>, rotation=<rot -0.0002 0 98.1346>}]
 cape_05_r [{shapeType=Capsule, length=14.3015, radius=10.5715, shapeName=cape_05_l_capsule, center=<9.0604 0.4838 3.5>, rotation=<rot -0.0003 0 84.3395>}]
 cape_06_l [{
   shapeType=Capsule
   length=31.2771
   radius=13.2018
   shapeName=cape_06_l_capsule
   center=<9.0604 -1.1673 3.5>
   rotation=<rot -0.0003 0 100.5198>
  }]
 cape_06_r [{
   shapeType=Capsule
   length=31.2771
   radius=13.2018
   shapeName=cape_06_l_capsule
   center=<9.0604 -1.1673 3.5>
   rotation=<rot -0.0005 0 78.6152>
  }]
 cape_07_l [{shapeType=Capsule, length=35.8061, radius=12.4031, shapeName=cape_07_l_capsule, center=<4.2502 0 3.5>, rotation=<rot -0.0001 0 99.2564>}]
 cape_07_r [{shapeType=Capsule, length=35.8061, radius=12.4031, shapeName=cape_07_l_capsule, center=<4.2502 0 3.5>, rotation=<rot -0.0002 0 82.4145>}]
 cape_l_MSTR [{shapeType=Capsule, length=1, radius=1, shapeName=cape_l_MSTR_capsule}]
 cape_mid_01_l [{shapeType=Capsule, length=9.8657, radius=9.4946, shapeName=cape_mid_01_l_capsule, center=<9.0604 0 11.5>, rotation=<rot 90 0 0>}]
 cape_mid_01_r [{shapeType=Capsule, length=9.8657, radius=9.4946, shapeName=cape_mid_01_r_capsule, center=<9.0604 0 11.5>, rotation=<rot 90 0 0>}]
 cape_mid_02_l [{shapeType=Capsule, length=13.7901, radius=9.6584, shapeName=cape_mid_02_l_capsule, center=<9.0604 0 6.3148>, rotation=<rot 90 0 0>}]
 cape_mid_02_r [{shapeType=Capsule, length=12.7901, radius=10.6584, shapeName=cape_mid_02_r_capsule, center=<9.0604 0 6.3148>, rotation=<rot 90 0 0>}]
 cape_mid_03_l [{
   shapeType=Capsule
   length=18.8322
   radius=9.6584
   shapeName=cape_mid_03_l_capsule
   center=<13.1914 0.0051 6.3148>
   rotation=<rot 90 172.875 -97.1961>
  }]
 cape_mid_03_r [{
   shapeType=Capsule
   length=8.0095
   radius=9.6584
   shapeName=cape_mid_03_r_capsule
   center=<13.1914 -0.0062 6.3148>
   rotation=<rot 90 -172.875 -82.7886>
  }]
 cape_mid_04_l [{shapeType=Capsule, length=14.2326, radius=9.6584, shapeName=cape_mid_04_l_capsule, center=<9.0604 0 6.3148>, rotation=<rot 0 0 90>}]
 cape_mid_04_r [{shapeType=Capsule, length=14.2326, radius=9.6584, shapeName=cape_mid_04_r_capsule, center=<9.0604 0 6.3148>, rotation=<rot 0 0 90>}]
 cape_mid_05_l [{shapeType=Capsule, length=28.66, radius=10.5715, shapeName=cape_mid_05_l_capsule, center=<9.0604 0 3.5>, rotation=<rot 0 0 90>}]
 cape_mid_05_r [{shapeType=Capsule, length=28.66, radius=10.5715, shapeName=cape_mid_05_r_capsule, center=<9.0604 0 3.5>, rotation=<rot 0 0 90>}]
 cape_mid_06_l [{shapeType=Capsule, length=43.992, radius=12.4084, shapeName=cape_mid_06_l_capsule, center=<9.0604 0 3.5>, rotation=<rot 0 0 90>}]
 cape_mid_06_r [{shapeType=Capsule, length=43.992, radius=12.4084, shapeName=cape_mid_06_r_capsule, center=<9.0604 0 3.5>, rotation=<rot 0 0 90>}]
 cape_mid_07_l [{shapeType=Capsule, length=46.3591, radius=13.1471, shapeName=cape_mid_07_l_capsule, center=<6.3942 0 3.5>, rotation=<rot 0 0 90>}]
 cape_mid_07_r [{shapeType=Capsule, length=46.3591, radius=13.1471, shapeName=cape_mid_07_r_capsule, center=<6.3942 0 3.5>, rotation=<rot 0 0 90>}]
 cape_mid_l_MSTR [{shapeType=Capsule, length=1, radius=1, shapeName=cape_mid_l_MSTR_capsule}]
 cape_mid_r_MSTR [{shapeType=Capsule, length=1, radius=1, shapeName=cape_mid_r_MSTR_capsule}]
 cape_r_MSTR [{shapeType=Capsule, length=1, radius=1, shapeName=cape_r_MSTR_capsule}]
 clavicle_l [{
   shapeType=Capsule
   length=8.9814
   radius=7.839
   shapeName=clavicle_l_capsule
   center=<10.6504 -0.7821 0>
   rotation=<rot 73.9892 -0.0009 -16.2994>
  }]
 clavicle_r [{
   shapeType=Capsule
   length=8.9814
   radius=7.839
   shapeName=clavicle_l_capsule
   center=<-10.6504 0.7821 0>
   rotation=<rot 73.9892 -0.0009 163.7006>
  }]
 dyn_lightSaber_01 [{shapeType=Capsule, length=2.6308, radius=2.1785, shapeName=dyn_lightSaber_01_capsule, center=<0.9373 0 2>, rotation=<rot 90 0 0>}]
 dyn_lightSaber_02 [{
   shapeType=Capsule
   length=26.2999
   radius=3.0585
   shapeName=dyn_lightSaber_02_capsule
   center=<11.9262 -0.4242 2.7795>
   rotation=<rot 90 0 0>
  }]
 dyn_lightSaber_MSTR [{shapeType=Capsule, length=0.505, radius=0.505, shapeName=dyn_lightSaber_MSTR_capsule, rotation=<rot 0 -90 0>}]
 foot_l [{
   shapeType=Capsule
   length=11.1718
   radius=15.0722
   shapeName=foot_l_capsule
   center=<-4.4416 6.7151 0>
   rotation=<rot 88.029 89.9992 -0.0006>
  }]
 foot_r [{
   shapeType=Capsule
   length=11.1718
   radius=15.0722
   shapeName=foot_l_capsule
   center=<4.4416 -6.7151 0>
   rotation=<rot 88.029 89.9992 179.9994>
  }]
 lowerarm_l [{shapeType=Capsule, length=21.1522, radius=5.5946, shapeName=lowerarm_l_capsule, center=<10.5383 0 0>, rotation=<rot 90 0 0>}]
 lowerarm_r [{shapeType=Capsule, length=21.1522, radius=5.5946, shapeName=lowerarm_l_capsule, center=<-10.5383 0 0>, rotation=<rot 90 0 180>}]
 pelvis [
  {shapeType=Capsule, length=8.3547, radius=13.4686, shapeName=pelvis_capsule}
  {shapeType=Capsule, length=9.0435, radius=13.5371, shapeName=pelvis_capsule, center=<-9.6022 0 0>}
 ]
 spine_01 [{shapeType=Capsule, length=10.2884, radius=11.535, shapeName=spine_01_capsule}]
 spine_02 [{shapeType=Capsule, length=10.2884, radius=11.535, shapeName=spine_02_capsule}]
 spine_03 [{shapeType=Capsule, length=10.2884, radius=11.535, shapeName=spine_03_capsule}]
 spine_04 [{shapeType=Capsule, length=10.2884, radius=11.535, shapeName=spine_04_capsule}]
 spine_05 [{shapeType=Capsule, length=16.0508, radius=11.3121, shapeName=spine_05_capsule}]
 thigh_l [{shapeType=Capsule, length=26.4437, radius=9.5433, shapeName=thigh_l_capsule, center=<-21.1879 0 0.9254>, rotation=<rot 90 0 0>}]
 thigh_r [{shapeType=Capsule, length=24.3818, radius=11.6052, shapeName=thigh_l_capsule, center=<21.1879 0 0>, rotation=<rot 90 0 180>}]
 upperarm_l [{shapeType=Capsule, length=20.4503, radius=6.1564, shapeName=upperarm_l_capsule, center=<11.2362 0 0>, rotation=<rot 90 0 0>}]
 upperarm_r [{shapeType=Capsule, length=20.4503, radius=6.1564, shapeName=upperarm_l_capsule, center=<-11.2362 0 0>, rotation=<rot 90 0 180>}]
}
constraints [
 {
  bone1Name=spine_01
  bone2Name=pelvis
  swing1LimitDegrees=45
  swing1Motion=Free
  swing2LimitDegrees=45
  swing2Motion=Free
  twistLimitDegrees=45
  twistMotion=Free
 }
 {
  bone1Name=spine_02
  bone2Name=spine_01
  swing1LimitDegrees=45
  swing1Motion=Free
  swing2LimitDegrees=45
  swing2Motion=Free
  twistLimitDegrees=45
  twistMotion=Free
 }
 {
  bone1Name=spine_03
  bone2Name=spine_02
  swing1LimitDegrees=45
  swing1Motion=Free
  swing2LimitDegrees=45
  swing2Motion=Free
  twistLimitDegrees=45
  twistMotion=Free
 }
 {
  bone1Name=spine_04
  bone2Name=spine_03
  swing1LimitDegrees=45
  swing1Motion=Free
  swing2LimitDegrees=45
  swing2Motion=Free
  twistLimitDegrees=45
  twistMotion=Free
 }
 {
  bone1Name=spine_05
  bone2Name=spine_04
  swing1LimitDegrees=45
  swing1Motion=Free
  swing2LimitDegrees=45
  swing2Motion=Free
  twistLimitDegrees=45
  twistMotion=Free
 }
 {bone1Name=cape_mid_01_l, bone2Name=cape_mid_l_MSTR, swing1LimitDegrees=5, swing2LimitDegrees=15, twistLimitDegrees=0}
 {bone1Name=cape_mid_02_l, bone2Name=cape_mid_01_l, swing1LimitDegrees=5, swing2LimitDegrees=30, twistLimitDegrees=0}
 {bone1Name=cape_mid_03_l, bone2Name=cape_mid_02_l, swing1LimitDegrees=5, swing2LimitDegrees=30, twistLimitDegrees=0}
 {bone1Name=cape_mid_04_l, bone2Name=cape_mid_03_l, swing1LimitDegrees=5, swing2LimitDegrees=30, twistLimitDegrees=0}
 {bone1Name=cape_mid_05_l, bone2Name=cape_mid_04_l, swing1LimitDegrees=5, swing2LimitDegrees=30, twistLimitDegrees=0}
 {bone1Name=cape_mid_06_l, bone2Name=cape_mid_05_l, swing1LimitDegrees=5, swing2LimitDegrees=30, twistLimitDegrees=0}
 {bone1Name=cape_mid_07_l, bone2Name=cape_mid_06_l, swing1LimitDegrees=5, swing2LimitDegrees=30, twistLimitDegrees=0}
 {bone1Name=cape_mid_01_r, bone2Name=cape_mid_r_MSTR, swing1LimitDegrees=5, swing2LimitDegrees=15, twistLimitDegrees=0}
 {bone1Name=cape_mid_02_r, bone2Name=cape_mid_01_r, swing1LimitDegrees=5, swing2LimitDegrees=30, twistLimitDegrees=0}
 {bone1Name=cape_mid_03_r, bone2Name=cape_mid_02_r, swing1LimitDegrees=5, swing2LimitDegrees=30, twistLimitDegrees=0}
 {bone1Name=cape_mid_04_r, bone2Name=cape_mid_03_r, swing1LimitDegrees=5, swing2LimitDegrees=30, twistLimitDegrees=0}
 {bone1Name=cape_mid_05_r, bone2Name=cape_mid_04_r, swing1LimitDegrees=5, swing2LimitDegrees=30, twistLimitDegrees=0}
 {bone1Name=cape_mid_06_r, bone2Name=cape_mid_05_r, swing1LimitDegrees=5, swing2LimitDegrees=30, twistLimitDegrees=0}
 {bone1Name=cape_mid_07_r, bone2Name=cape_mid_06_r, swing1LimitDegrees=5, swing2LimitDegrees=30, twistLimitDegrees=0}
 {
  bone1Name=cape_mid_r_MSTR
  bone2Name=spine_05
  swing1LimitDegrees=45
  swing1Motion=Free
  swing2LimitDegrees=45
  swing2Motion=Free
  twistLimitDegrees=45
  twistMotion=Free
 }
 {
  bone1Name=cape_mid_l_MSTR
  bone2Name=spine_05
  swing1LimitDegrees=45
  swing1Motion=Free
  swing2LimitDegrees=45
  swing2Motion=Free
  twistLimitDegrees=45
  twistMotion=Free
 }
 {
  bone1Name=cape_l_MSTR
  bone2Name=clavicle_l
  swing1LimitDegrees=45
  swing1Motion=Free
  swing2LimitDegrees=45
  swing2Motion=Free
  twistLimitDegrees=45
  twistMotion=Free
 }
 {bone1Name=cape_01_l, bone2Name=cape_l_MSTR, swing1LimitDegrees=15, swing2LimitDegrees=30, twistLimitDegrees=0}
 {bone1Name=cape_02_l, bone2Name=cape_01_l, swing1LimitDegrees=15, swing2LimitDegrees=30, twistLimitDegrees=0}
 {bone1Name=cape_03_l, bone2Name=cape_02_l, swing1LimitDegrees=15, swing2LimitDegrees=30, twistLimitDegrees=0}
 {bone1Name=cape_04_l, bone2Name=cape_03_l, swing1LimitDegrees=20, swing2LimitDegrees=20, twistLimitDegrees=0}
 {bone1Name=cape_05_l, bone2Name=cape_04_l, swing1LimitDegrees=20, swing2LimitDegrees=20, twistLimitDegrees=0}
 {bone1Name=cape_06_l, bone2Name=cape_05_l, swing1LimitDegrees=20, swing2LimitDegrees=20, twistLimitDegrees=0}
 {bone1Name=cape_07_l, bone2Name=cape_06_l, swing1LimitDegrees=20, swing2LimitDegrees=20, twistLimitDegrees=0}
 {
  bone1Name=upperarm_l
  bone2Name=clavicle_l
  swing1LimitDegrees=45
  swing1Motion=Free
  swing2LimitDegrees=45
  swing2Motion=Free
  twistLimitDegrees=45
  twistMotion=Free
 }
 {
  bone1Name=lowerarm_l
  bone2Name=upperarm_l
  swing1LimitDegrees=45
  swing1Motion=Free
  swing2LimitDegrees=45
  swing2Motion=Free
  twistLimitDegrees=45
  twistMotion=Free
 }
 {
  bone1Name=cape_r_MSTR
  bone2Name=clavicle_r
  swing1LimitDegrees=45
  swing1Motion=Free
  swing2LimitDegrees=45
  swing2Motion=Free
  twistLimitDegrees=45
  twistMotion=Free
 }
 {bone1Name=cape_01_r, bone2Name=cape_r_MSTR, swing1LimitDegrees=15, swing2LimitDegrees=30, twistLimitDegrees=0}
 {bone1Name=cape_02_r, bone2Name=cape_01_r, swing1LimitDegrees=20, swing2LimitDegrees=30, twistLimitDegrees=0}
 {bone1Name=cape_03_r, bone2Name=cape_02_r, swing1LimitDegrees=20, swing2LimitDegrees=30, twistLimitDegrees=0}
 {bone1Name=cape_04_r, bone2Name=cape_03_r, swing1LimitDegrees=30, swing2LimitDegrees=30, twistLimitDegrees=0}
 {bone1Name=cape_05_r, bone2Name=cape_04_r, swing1LimitDegrees=30, swing2LimitDegrees=30, twistLimitDegrees=0}
 {bone1Name=cape_06_r, bone2Name=cape_05_r, swing1LimitDegrees=30, swing2LimitDegrees=30, twistLimitDegrees=0}
 {bone1Name=cape_07_r, bone2Name=cape_06_r, swing1LimitDegrees=30, swing2LimitDegrees=30, twistLimitDegrees=0}
 {
  bone1Name=clavicle_l
  bone2Name=spine_05
  swing1LimitDegrees=45
  swing1Motion=Free
  swing2LimitDegrees=45
  swing2Motion=Free
  twistLimitDegrees=45
  twistMotion=Free
 }
 {
  bone1Name=calf_l
  bone2Name=thigh_l
  swing1LimitDegrees=45
  swing1Motion=Free
  swing2LimitDegrees=45
  swing2Motion=Free
  twistLimitDegrees=45
  twistMotion=Free
 }
 {
  bone1Name=thigh_l
  bone2Name=pelvis
  swing1LimitDegrees=45
  swing1Motion=Free
  swing2LimitDegrees=45
  swing2Motion=Free
  twistLimitDegrees=45
  twistMotion=Free
 }
 {
  bone1Name=foot_l
  bone2Name=calf_l
  swing1LimitDegrees=45
  swing1Motion=Free
  swing2LimitDegrees=45
  swing2Motion=Free
  twistLimitDegrees=45
  twistMotion=Free
 }
 {
  bone1Name=calf_r
  bone2Name=thigh_r
  swing1LimitDegrees=45
  swing1Motion=Free
  swing2LimitDegrees=45
  swing2Motion=Free
  twistLimitDegrees=45
  twistMotion=Free
 }
 {
  bone1Name=foot_r
  bone2Name=calf_r
  swing1LimitDegrees=45
  swing1Motion=Free
  swing2LimitDegrees=45
  swing2Motion=Free
  twistLimitDegrees=45
  twistMotion=Free
 }
 {
  bone1Name=thigh_r
  bone2Name=pelvis
  swing1LimitDegrees=45
  swing1Motion=Free
  swing2LimitDegrees=45
  swing2Motion=Free
  twistLimitDegrees=45
  twistMotion=Free
 }
 {
  bone1Name=upperarm_r
  bone2Name=clavicle_r
  swing1LimitDegrees=45
  swing1Motion=Free
  swing2LimitDegrees=45
  swing2Motion=Free
  twistLimitDegrees=45
  twistMotion=Free
 }
 {
  bone1Name=lowerarm_r
  bone2Name=upperarm_r
  swing1LimitDegrees=45
  swing1Motion=Free
  swing2LimitDegrees=45
  swing2Motion=Free
  twistLimitDegrees=45
  twistMotion=Free
 }
 {
  bone1Name=clavicle_r
  bone2Name=spine_05
  swing1LimitDegrees=45
  swing1Motion=Free
  swing2LimitDegrees=45
  swing2Motion=Free
  twistLimitDegrees=45
  twistMotion=Free
 }
 {bone1Name=dyn_lightSaber_02, bone2Name=dyn_lightSaber_01, swing1LimitDegrees=15, swing2LimitDegrees=15, twistLimitDegrees=0}
 {
  bone1Name=dyn_lightSaber_MSTR
  bone2Name=pelvis
  swing1LimitDegrees=45
  swing1Motion=Free
  swing2LimitDegrees=45
  swing2Motion=Free
  twistLimitDegrees=45
  twistMotion=Free
 }
 {bone1Name=dyn_lightSaber_01, bone2Name=dyn_lightSaber_MSTR, swing1LimitDegrees=15, swing2LimitDegrees=15, twistLimitDegrees=0}
]
