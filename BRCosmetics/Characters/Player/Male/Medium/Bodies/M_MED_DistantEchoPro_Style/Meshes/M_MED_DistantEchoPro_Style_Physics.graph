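# Physics asset for the DistantEchoPro medium male player body style, covering a lightsaber prop that hangs from the pelvis.
# Capsule bodies form a dyn_lightSaber chain from a small master body through two saber segments, alongside pelvis, lower spine and left thigh capsules that approximate the nearby body.
# Master, spine and thigh constraints rotate freely, while the two saber segment joints are tighter, with zero twist limit.
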PhysicsAsset /BRCosmetics/Characters/Player/Male/Medium/Bodies/M_MED_DistantEchoPro_Style/Meshes/M_MED_DistantEchoPro_Style_Physics
bodies {
 dyn_lightSaber_01 [{shapeType=Capsule, length=2.6308, radius=2.1785, shapeName=dyn_lightSaber_01_capsule, center=<0.9373 0 2>, rotation=<rot 90 0 0>}]
 dyn_lightSaber_02 [{shapeType=Capsule, length=25.3516, radius=2.1785, shapeName=dyn_lightSaber_02_capsule, center=<11.3134 0 2>, rotation=<rot 90 0 0>}]
 dyn_lightSaber_MSTR [{shapeType=Capsule, length=0.5, radius=0.5, shapeName=dyn_lightSaber_MSTR_capsule}]
 pelvis [{shapeType=Capsule, length=5, radius=12.5025, shapeName=pelvis_capsule}]
 spine_01 [{shapeType=Capsule, length=9.7054, radius=9.2067, shapeName=spine_01_capsule}]
 spine_02 [{shapeType=Capsule, length=9.7054, radius=9.2067, shapeName=spine_02_capsule}]
 thigh_l [{
   shapeType=Capsule
   length=25.654
   radius=9.8917
   shapeName=thigh_l_capsule
   center=<-10.4025 0.0001 0.4532>
   rotation=<rot 82.5832 179.9979 179.9995>
  }]
}
constraints [
 {
  bone1Name=spine_01
  bone2Name=pelvis
  swing1LimitDegrees=45
  swing1Motion=Free
  swing2LimitDegrees=45
  swing2Motion=Free
  twistLimitDegrees=45
  twistMotion=Free
 }
 {
  bone1Name=spine_02
  bone2Name=spine_01
  swing1LimitDegrees=45
  swing1Motion=Free
  swing2LimitDegrees=45
  swing2Motion=Free
  twistLimitDegrees=45
  twistMotion=Free
 }
 {
  bone1Name=thigh_l
  bone2Name=pelvis
  swing1LimitDegrees=45
  swing1Motion=Free
  swing2LimitDegrees=45
  swing2Motion=Free
  twistLimitDegrees=45
  twistMotion=Free
 }
 {bone1Name=dyn_lightSaber_01, bone2Name=dyn_lightSaber_MSTR, swing1LimitDegrees=45, swing2LimitDegrees=45, twistLimitDegrees=0}
 {bone1Name=dyn_lightSaber_02, bone2Name=dyn_lightSaber_01, swing1LimitDegrees=45, swing2LimitDegrees=45, twistLimitDegrees=0}
 {
  bone1Name=dyn_lightSaber_MSTR
  bone2Name=pelvis
  swing1LimitDegrees=45
  swing1Motion=Free
  swing2LimitDegrees=45
  swing2Motion=Free
  twistLimitDegrees=45
  twistMotion=Free
 }
]
