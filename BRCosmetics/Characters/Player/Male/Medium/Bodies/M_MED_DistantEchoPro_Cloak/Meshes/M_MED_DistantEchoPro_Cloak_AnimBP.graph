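# Post-process-style cosmetic AnimBP for the Distant Echo Pro cloak, copying the parent mesh pose through style and cloak Control Rigs.
# RigidBody simulations with separate physics assets cover standard, skydive, backbling, crouch and sliding cases, selected by bool blends.
# Bool-driven bone offsets and constraints correct the cape per movement and emote state, and a LookAt aims the lightsaber bone.
# The EventGraph updates gravity override, backbling and weapon-related state.
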AnimBlueprint /BRCosmetics/Characters/Player/Male/Medium/Bodies/M_MED_DistantEchoPro_Cloak/Meshes/M_MED_DistantEchoPro_Cloak_AnimBP
target_skeleton=&/M_MED_DistantEchoPro_Cloak_Skeleton
graphs [
 AnimGraph {nodes [
   n1 = Get bIsSprinting()
   n2 = "Transform (Modify) Bone - Bone: cape_mid_03_l"!(ComponentPose, bAlphaBoolEnabled:bool=@n1, Translation="0,0,0", Rotation="7,0,0", Scale="1,1,1") -> (Pose) {
    alphaInputType=Bool
    rotationMode=BMM_Additive
    rotationSpace=BCS_BoneSpace
    translationMode=BMM_Additive
    translationSpace=BCS_BoneSpace
    alphaBoolBlend {blendInTime=0.4, blendOption=Linear, blendOutTime=0.5}
    boneToModify {boneName=cape_mid_03_l}
    rotation {pitch=7}
   }
   n3 = "Transform (Modify) Bone - Bone: cape_mid_03_r"!(ComponentPose=@n2, bAlphaBoolEnabled:bool=@n1, Translation="0,0,0", Rotation="7,0,0", Scale="1,1,1") -> (Pose) {
    alphaInputType=Bool
    rotationMode=BMM_Additive
    rotationSpace=BCS_BoneSpace
    translationMode=BMM_Additive
    translationSpace=BCS_BoneSpace
    alphaBoolBlend {blendInTime=0.4, blendOption=Linear, blendOutTime=0.5}
    boneToModify {boneName=cape_mid_03_r}
    rotation {pitch=7}
   }
   n4 = "Transform (Modify) Bone - Bone: cape_mid_04_l"!(ComponentPose=@n3, bAlphaBoolEnabled:bool=@n1, Translation="0,0,0", Rotation="12,0,0", Scale="1,1,1") -> (Pose) {
    alphaInputType=Bool
    rotationMode=BMM_Additive
    rotationSpace=BCS_BoneSpace
    translationMode=BMM_Additive
    translationSpace=BCS_BoneSpace
    alphaBoolBlend {blendInTime=0.4, blendOption=Linear, blendOutTime=0.5}
    boneToModify {boneName=cape_mid_04_l}
    rotation {pitch=12}
   }
   n5 = "Transform (Modify) Bone - Bone: cape_mid_04_r"!(ComponentPose=@n4, bAlphaBoolEnabled:bool=@n1, Translation="0,0,0", Rotation="7,0,0", Scale="1,1,1") -> (Pose) {
    alphaInputType=Bool
    rotationMode=BMM_Additive
    rotationSpace=BCS_BoneSpace
    translationMode=BMM_Additive
    translationSpace=BCS_BoneSpace
    alphaBoolBlend {blendInTime=0.4, blendOption=Linear, blendOutTime=0.5}
    boneToModify {boneName=cape_mid_04_r}
    rotation {pitch=7}
   }
   n6 = "Transform (Modify) Bone - Bone: cape_03_l"!(ComponentPose=@n5, bAlphaBoolEnabled:bool=@n1, Translation="0,0,0", Rotation="12,0,0", Scale="1,1,1") -> (Pose) {
    alphaInputType=Bool
    rotationMode=BMM_Additive
    rotationSpace=BCS_BoneSpace
    translationMode=BMM_Additive
    translationSpace=BCS_BoneSpace
    alphaBoolBlend {blendInTime=0.4, blendOption=Linear, blendOutTime=0.5}
    boneToModify {boneName=cape_03_l}
    rotation {pitch=12}
   }
   "Transform (Modify) Bone - Bone: cape_mid_04_r"(ComponentPose=@n6, bAlphaBoolEnabled=@n1, Translation="0,0,0", Rotation="0,0,0", Scale="1,1,1") {
    alphaInputType=Bool
    rotationMode=BMM_Additive
    rotationSpace=BCS_BoneSpace
    translationMode=BMM_Additive
    translationSpace=BCS_BoneSpace
    alphaBoolBlend {blendInTime=0.4, blendOption=Linear, blendOutTime=0.5}
    boneToModify {boneName=cape_mid_04_r}
   }
   n7 = "Transform (Modify) Bone - Bone: cape_mid_02_l"!(ComponentPose, bAlphaBoolEnabled:bool=True, Translation="0,0,0", Rotation="-30,0,0", Scale="1,1,1") -> (Pose) {
    alphaInputType=Bool
    rotationMode=BMM_Additive
    rotationSpace=BCS_BoneSpace
    alphaBoolBlend {blendOption=Linear}
    boneToModify {boneName=cape_mid_02_l}
    rotation {pitch=-30}
   }
   "Transform (Modify) Bone - Bone: cape_mid_02_r"!(ComponentPose=@n7, bAlphaBoolEnabled:bool=True, Translation="0,0,0", Rotation="-30,0,0", Scale="1,1,1") -> (Pose) {
    alphaInputType=Bool
    rotationMode=BMM_Additive
    rotationSpace=BCS_BoneSpace
    alphaBoolBlend {blendOption=Linear}
    boneToModify {boneName=cape_mid_02_r}
    rotation {pitch=-30}
   }
   RigidBody!(ComponentPose, Alpha, OverrideWorldGravity:Vector=Get Gravity Override(), ExternalForce) -> (Pose) {
    bClampLinearTranslationLimitToRefPose=true
    bOverrideWorldGravity=true
    overridePhysicsAsset=/BRCosmetics/Characters/Player/Male/Medium/Bodies/M_MED_DistantEchoPro_Style/Meshes/M_MED_DistantEchoPro_Style_Physics
    simulationSpace=BaseBoneSpace
    alphaBoolBlend {blendOption=Linear}
    baseBoneRef {boneName=pelvis}
    componentAppliedLinearAccClamp=<6000 6000 6000>
    componentLinearAccScale=<0.1 0.5 0.5>
    componentLinearVelScale=<0.1 0.5 0.5>
    overrideWorldGravity {z=-980}
   }
   n8 = Get bIsSkydiveFloating()
   n9 = "Constraint - cape_01_l (P)"!(ComponentPose=Local To Component(LocalPose=Use cached pose 'simCache'()), bAlphaBoolEnabled:bool=@n8, ConstraintWeights_0=1) -> (Pose) {
    alphaInputType=Bool
    alphaBoolBlend {blendInTime=0.5, blendOption=Linear, blendOutTime=0.5}
    boneToModify {boneName=cape_01_l}
    constraintSetup [{targetBone {boneName=spine_05}}]
    constraintWeights [1]
   }
   n10 = "Constraint - cape_01_r (P)"!(ComponentPose=@n9, bAlphaBoolEnabled:bool=@n8, ConstraintWeights_0=1) -> (Pose) {
    alphaInputType=Bool
    alphaBoolBlend {blendInTime=0.5, blendOption=Linear, blendOutTime=0.5}
    boneToModify {boneName=cape_01_r}
    constraintSetup [{targetBone {boneName=spine_05}}]
    constraintWeights [1]
   }
   n11 = "Transform (Modify) Bone - Bone: cape_01_l"!(ComponentPose=@n10, bAlphaBoolEnabled:bool=@n8, Translation="5,0,10", Rotation="0,0,0", Scale="1,1,1") -> (Pose) {
    alphaInputType=Bool
    rotationMode=BMM_Additive
    rotationSpace=BCS_BoneSpace
    translationMode=BMM_Additive
    translationSpace=BCS_BoneSpace
    alphaBoolBlend {blendInTime=0.2, blendOption=Linear, blendOutTime=0.2}
    boneToModify {boneName=cape_01_l}
    translation {x=5, z=10}
   }
   n12 = "Transform (Modify) Bone - Bone: cape_01_r"!(ComponentPose=@n11, bAlphaBoolEnabled:bool=@n8, Translation="0,0,10", Rotation="20,0,0", Scale="1,1,1") -> (Pose) {
    alphaInputType=Bool
    rotationMode=BMM_Additive
    rotationSpace=BCS_BoneSpace
    translationMode=BMM_Additive
    translationSpace=BCS_BoneSpace
    alphaBoolBlend {blendInTime=0.2, blendOption=Linear, blendOutTime=0.2}
    boneToModify {boneName=cape_01_r}
    rotation {pitch=20}
    translation {z=10}
   }
   n13 = Get bIsSkydiveFloating()
   n14 = "Transform (Modify) Bone - Bone: dyn_capeShoulder_l_01"!(ComponentPose=@n12, bAlphaBoolEnabled:bool=@n13, Translation="-4,3,0", Rotation="0,-10,20", Scale="1,1,1") -> (Pose) {
    alphaInputType=Bool
    rotationMode=BMM_Additive
    rotationSpace=BCS_BoneSpace
    translationMode=BMM_Additive
    translationSpace=BCS_BoneSpace
    alphaBoolBlend {blendInTime=0.2, blendOption=Linear, blendOutTime=0.2}
    boneToModify {boneName=dyn_capeShoulder_l_01}
    rotation {roll=20, yaw=-10}
    translation=<-4 3>
   }
   n15 = "Transform (Modify) Bone - Bone: dyn_capeShoulder_r_01"!(ComponentPose=@n14, bAlphaBoolEnabled:bool=@n13, Translation="4,-3,0", Rotation="0,-15,20", Scale="1,1,1") -> (Pose) {
    alphaInputType=Bool
    rotationMode=BMM_Additive
    rotationSpace=BCS_BoneSpace
    translationMode=BMM_Additive
    translationSpace=BCS_BoneSpace
    alphaBoolBlend {blendInTime=0.2, blendOption=Linear, blendOutTime=0.2}
    boneToModify {boneName=dyn_capeShoulder_r_01}
    rotation {roll=20, yaw=-15}
    translation=<4 -3>
   }
   n16 = "Transform (Modify) Bone - Bone: cape_mid_01_l"!(ComponentPose=@n15, bAlphaBoolEnabled:bool=@n8, Translation="0,0,5", Rotation="0,0,0", Scale="1,1,1") -> (Pose) {
    alphaInputType=Bool
    rotationMode=BMM_Additive
    rotationSpace=BCS_BoneSpace
    translationMode=BMM_Additive
    translationSpace=BCS_BoneSpace
    alphaBoolBlend {blendOption=Linear}
    boneToModify {boneName=cape_mid_01_l}
    translation {z=5}
   }
   n17 = "Transform (Modify) Bone - Bone: cape_mid_01_r"!(ComponentPose=@n16, bAlphaBoolEnabled:bool=@n8, Translation="0,0,5", Rotation="0,0,0", Scale="1,1,1") -> (Pose) {
    alphaInputType=Bool
    rotationMode=BMM_Additive
    rotationSpace=BCS_BoneSpace
    translationMode=BMM_Additive
    translationSpace=BCS_BoneSpace
    alphaBoolBlend {blendOption=Linear}
    boneToModify {boneName=cape_mid_01_r}
    translation {z=5}
   }
   n18 = "Constraint - cape_01_l (P)"(ComponentPose=@n17, bAlphaBoolEnabled=false, ConstraintWeights_0=1) {
    alphaInputType=Bool
    alphaBoolBlend {blendInTime=0.1, blendOption=Linear, blendOutTime=0.1}
    boneToModify {boneName=cape_01_l}
    constraintSetup [{targetBone {boneName=upperarm_l}}]
    constraintWeights [1]
   }
   n19 = "Constraint - cape_01_r (P)"(ComponentPose=@n18, bAlphaBoolEnabled=false, ConstraintWeights_0=1) {
    alphaInputType=Bool
    alphaBoolBlend {blendInTime=0.1, blendOption=Linear, blendOutTime=0.1}
    boneToModify {boneName=cape_01_r}
    constraintSetup [{targetBone {boneName=upperarm_r}}]
    constraintWeights [1]
   }
   n20 = Get bIsSkydiveDiveMode()
   n21 = "Transform (Modify) Bone - Bone: cape_01_l"(ComponentPose=@n19, bAlphaBoolEnabled=@n20, Translation="0,-15,5", Rotation="0,0,0", Scale="1,1,1") {
    alphaInputType=Bool
    rotationMode=BMM_Additive
    rotationSpace=BCS_BoneSpace
    translationMode=BMM_Additive
    translationSpace=BCS_BoneSpace
    alphaBoolBlend {blendInTime=0.2, blendOption=Linear, blendOutTime=0.2}
    boneToModify {boneName=cape_01_l}
    translation {y=-15, z=5}
   }
   n22 = "Transform (Modify) Bone - Bone: cape_01_r"(ComponentPose=@n21, bAlphaBoolEnabled=@n20, Translation="0,15,5", Rotation="0,0,0", Scale="1,1,1") {
    alphaInputType=Bool
    rotationMode=BMM_Additive
    rotationSpace=BCS_BoneSpace
    translationMode=BMM_Additive
    translationSpace=BCS_BoneSpace
    alphaBoolBlend {blendInTime=0.2, blendOption=Linear, blendOutTime=0.2}
    boneToModify {boneName=cape_01_r}
    translation {y=15, z=5}
   }
   n23 = "Transform (Modify) Bone - Bone: cape_02_l"!(ComponentPose=@n22, bAlphaBoolEnabled:bool=@n20, Translation="0,10,0", Rotation="-10,0,0", Scale="1,1,1") -> (Pose) {
    alphaInputType=Bool
    rotationMode=BMM_Additive
    rotationSpace=BCS_BoneSpace
    translationMode=BMM_Additive
    translationSpace=BCS_BoneSpace
    alphaBoolBlend {blendInTime=0.2, blendOption=Linear, blendOutTime=0.2}
    boneToModify {boneName=cape_02_l}
    rotation {pitch=-10}
    translation {y=10}
   }
   n24 = "Transform (Modify) Bone - Bone: cape_02_r"!(ComponentPose=@n23, bAlphaBoolEnabled:bool=@n20, Translation="0,-10,0", Rotation="-10,0,0", Scale="1,1,1") -> (Pose) {
    alphaInputType=Bool
    rotationMode=BMM_Additive
    rotationSpace=BCS_BoneSpace
    translationMode=BMM_Additive
    translationSpace=BCS_BoneSpace
    alphaBoolBlend {blendInTime=0.2, blendOption=Linear, blendOutTime=0.2}
    boneToModify {boneName=cape_02_r}
    rotation {pitch=-10}
    translation {y=-10}
   }
   n25 = "Transform (Modify) Bone - Bone: cape_mid_02_l"(ComponentPose=@n24, bAlphaBoolEnabled=@n20, Translation="0,0,8", Rotation="0,0,0", Scale="1,1,1") {
    alphaInputType=Bool
    rotationMode=BMM_Additive
    rotationSpace=BCS_BoneSpace
    translationMode=BMM_Additive
    translationSpace=BCS_BoneSpace
    alphaBoolBlend {blendInTime=0.2, blendOption=Linear, blendOutTime=0.2}
    boneToModify {boneName=cape_mid_02_l}
    translation {z=8}
   }
   n26 = "Transform (Modify) Bone - Bone: cape_mid_02_r"(ComponentPose=@n25, bAlphaBoolEnabled=@n20, Translation="0,0,10", Rotation="0,0,0", Scale="1,1,1") {
    alphaInputType=Bool
    rotationMode=BMM_Additive
    rotationSpace=BCS_BoneSpace
    translationMode=BMM_Additive
    translationSpace=BCS_BoneSpace
    alphaBoolBlend {blendInTime=0.2, blendOption=Linear, blendOutTime=0.2}
    boneToModify {boneName=cape_mid_02_r}
    translation {z=10}
   }
   n27 = Get bIsParachuteOpened()
   n28 = "Transform (Modify) Bone - Bone: cape_01_l"(ComponentPose=@n26, bAlphaBoolEnabled=@n27, Translation="0,0,5", Rotation="5,0,0", Scale="1,1,1") {
    alphaInputType=Bool
    rotationMode=BMM_Additive
    rotationSpace=BCS_BoneSpace
    translationMode=BMM_Additive
    translationSpace=BCS_BoneSpace
    alphaBoolBlend {blendInTime=0.1, blendOption=Linear, blendOutTime=0.3}
    boneToModify {boneName=cape_01_l}
    rotation {pitch=5}
    translation {z=5}
   }
   n29 = "Transform (Modify) Bone - Bone: cape_01_r"(ComponentPose=@n28, bAlphaBoolEnabled=@n27, Translation="0,0,5", Rotation="5,0,0", Scale="1,1,1") {
    alphaInputType=Bool
    rotationMode=BMM_Additive
    rotationSpace=BCS_BoneSpace
    translationMode=BMM_Additive
    translationSpace=BCS_BoneSpace
    alphaBoolBlend {blendInTime=0.1, blendOption=Linear, blendOutTime=0.3}
    boneToModify {boneName=cape_01_r}
    rotation {pitch=5}
    translation {z=5}
   }
   n30 = "Transform (Modify) Bone - Bone: cape_02_l"(ComponentPose=@n29, bAlphaBoolEnabled=@n27, Translation="0,0,0", Rotation="5,0,0", Scale="1,1,1") {
    alphaInputType=Bool
    rotationMode=BMM_Additive
    rotationSpace=BCS_BoneSpace
    translationMode=BMM_Additive
    translationSpace=BCS_BoneSpace
    alphaBoolBlend {blendInTime=0.1, blendOption=Linear, blendOutTime=0.3}
    boneToModify {boneName=cape_02_l}
    rotation {pitch=5}
   }
   n31 = "Transform (Modify) Bone - Bone: cape_02_r"(ComponentPose=@n30, bAlphaBoolEnabled=@n27, Translation="0,0,0", Rotation="5,0,0", Scale="1,1,1") {
    alphaInputType=Bool
    rotationMode=BMM_Additive
    rotationSpace=BCS_BoneSpace
    translationMode=BMM_Additive
    translationSpace=BCS_BoneSpace
    alphaBoolBlend {blendInTime=0.1, blendOption=Linear, blendOutTime=0.3}
    boneToModify {boneName=cape_02_r}
    rotation {pitch=5}
   }
   n32 = "Transform (Modify) Bone - Bone: cape_mid_02_l"(ComponentPose=@n31, bAlphaBoolEnabled=@n27, Translation="0,0,0", Rotation="10,0,0", Scale="1,1,1") {
    alphaInputType=Bool
    rotationMode=BMM_Additive
    rotationSpace=BCS_BoneSpace
    translationMode=BMM_Additive
    translationSpace=BCS_BoneSpace
    alphaBoolBlend {blendInTime=0.1, blendOption=Linear, blendOutTime=0.3}
    boneToModify {boneName=cape_mid_02_l}
    rotation {pitch=10}
   }
   n33 = "Transform (Modify) Bone - Bone: cape_mid_02_r"(ComponentPose=@n32, bAlphaBoolEnabled=@n27, Translation="0,0,0", Rotation="10,0,0", Scale="1,1,1") {
    alphaInputType=Bool
    rotationMode=BMM_Additive
    rotationSpace=BCS_BoneSpace
    translationMode=BMM_Additive
    translationSpace=BCS_BoneSpace
    alphaBoolBlend {blendInTime=0.1, blendOption=Linear, blendOutTime=0.3}
    boneToModify {boneName=cape_mid_02_r}
    rotation {pitch=10}
   }
   n34 = "Transform (Modify) Bone - Bone: cape_04_l"!(ComponentPose=@n33, bAlphaBoolEnabled:bool=@n27, Translation="0,0,0", Rotation="10,0,0", Scale="1,1,1") -> (Pose) {
    alphaInputType=Bool
    rotationMode=BMM_Additive
    rotationSpace=BCS_BoneSpace
    translationMode=BMM_Additive
    translationSpace=BCS_BoneSpace
    alphaBoolBlend {blendInTime=0.1, blendOption=Linear, blendOutTime=0.3}
    boneToModify {boneName=cape_04_l}
    rotation {pitch=10}
   }
   n35 = "Transform (Modify) Bone - Bone: cape_04_r"!(ComponentPose=@n34, bAlphaBoolEnabled:bool=@n27, Translation="0,0,0", Rotation="10,0,0", Scale="1,1,1") -> (Pose) {
    alphaInputType=Bool
    rotationMode=BMM_Additive
    rotationSpace=BCS_BoneSpace
    translationMode=BMM_Additive
    translationSpace=BCS_BoneSpace
    alphaBoolBlend {blendInTime=0.1, blendOption=Linear, blendOutTime=0.3}
    boneToModify {boneName=cape_04_r}
    rotation {pitch=10}
   }
   n36 = "Transform (Modify) Bone - Bone: cape_mid_04_l"(ComponentPose=@n35, bAlphaBoolEnabled=@n27, Translation="0,0,0", Rotation="10,0,0", Scale="1,1,1") {
    alphaInputType=Bool
    rotationMode=BMM_Additive
    rotationSpace=BCS_BoneSpace
    translationMode=BMM_Additive
    translationSpace=BCS_BoneSpace
    alphaBoolBlend {blendInTime=0.1, blendOption=Linear, blendOutTime=0.3}
    boneToModify {boneName=cape_mid_04_l}
    rotation {pitch=10}
   }
   n37 = "Transform (Modify) Bone - Bone: cape_mid_04_r"(ComponentPose=@n36, bAlphaBoolEnabled=@n27, Translation="0,0,0", Rotation="10,0,0", Scale="1,1,1") {
    alphaInputType=Bool
    rotationMode=BMM_Additive
    rotationSpace=BCS_BoneSpace
    translationMode=BMM_Additive
    translationSpace=BCS_BoneSpace
    alphaBoolBlend {blendInTime=0.1, blendOption=Linear, blendOutTime=0.3}
    boneToModify {boneName=cape_mid_04_r}
    rotation {pitch=10}
   }
   n38 = Get bIsCrouching()
   n39 = "Constraint - cape_mid_02_l (P)"!(ComponentPose=@n37, bAlphaBoolEnabled:bool=@n38, ConstraintWeights_0=1) -> (Pose) {
    alphaInputType=Bool
    alphaBoolBlend {blendOption=Linear}
    boneToModify {boneName=cape_mid_02_l}
    constraintSetup [{targetBone {boneName=spine_04}}]
    constraintWeights [1]
   }
   n40 = "Constraint - cape_mid_02_r (P)"!(ComponentPose=@n39, bAlphaBoolEnabled:bool=@n38, ConstraintWeights_0=1) -> (Pose) {
    alphaInputType=Bool
    alphaBoolBlend {blendOption=Linear}
    boneToModify {boneName=cape_mid_02_r}
    constraintSetup [{targetBone {boneName=spine_04}}]
    constraintWeights [1]
   }
   n41 = Get isCrouchWalking() -> (isCrouchWalking:bool)
   n42 = "Transform (Modify) Bone - Bone: cape_mid_01_l"(ComponentPose=@n40, bAlphaBoolEnabled=@n41, Translation="0,0,0", Rotation="-5,0,0", Scale="1,1,1") {
    alphaInputType=Bool
    rotationMode=BMM_Additive
    rotationSpace=BCS_BoneSpace
    translationMode=BMM_Additive
    translationSpace=BCS_BoneSpace
    alphaBoolBlend {blendInTime=0.2, blendOption=Linear, blendOutTime=0.2}
    boneToModify {boneName=cape_mid_01_l}
    rotation {pitch=-5}
   }
   n43 = "Transform (Modify) Bone - Bone: cape_mid_02_r"(ComponentPose=@n42, bAlphaBoolEnabled=@n41, Translation="0,0,0", Rotation="-10,10,0", Scale="1,1,1") {
    alphaInputType=Bool
    rotationMode=BMM_Additive
    rotationSpace=BCS_BoneSpace
    translationMode=BMM_Additive
    translationSpace=BCS_BoneSpace
    alphaBoolBlend {blendInTime=0.2, blendOption=Linear, blendOutTime=0.2}
    boneToModify {boneName=cape_mid_02_r}
    rotation {pitch=-10, yaw=10}
   }
   n44 = "Transform (Modify) Bone - Bone: cape_mid_03_l"(ComponentPose=@n43, bAlphaBoolEnabled=@n41, Translation="0,0,0", Rotation="-5,0,0", Scale="1,1,1") {
    alphaInputType=Bool
    rotationMode=BMM_Additive
    rotationSpace=BCS_BoneSpace
    translationMode=BMM_Additive
    translationSpace=BCS_BoneSpace
    alphaBoolBlend {blendInTime=0.2, blendOption=Linear, blendOutTime=0.2}
    boneToModify {boneName=cape_mid_03_l}
    rotation {pitch=-5}
   }
   n45 = "Transform (Modify) Bone - Bone: cape_mid_03_r"(ComponentPose=@n44, bAlphaBoolEnabled=@n41, Translation="0,0,0", Rotation="-5,0,0", Scale="1,1,1") {
    alphaInputType=Bool
    rotationMode=BMM_Additive
    rotationSpace=BCS_BoneSpace
    translationMode=BMM_Additive
    translationSpace=BCS_BoneSpace
    alphaBoolBlend {blendInTime=0.2, blendOption=Linear, blendOutTime=0.2}
    boneToModify {boneName=cape_mid_03_r}
    rotation {pitch=-5}
   }
   n46 = "Transform (Modify) Bone - Bone: cape_02_r"(ComponentPose=@n45, bAlphaBoolEnabled=@n41, Translation="0,3,1.5", Rotation="0,0,0", Scale="1,1,1") {
    alphaInputType=Bool
    rotationMode=BMM_Additive
    rotationSpace=BCS_BoneSpace
    translationMode=BMM_Additive
    translationSpace=BCS_BoneSpace
    alphaBoolBlend {blendInTime=0.2, blendOption=Linear, blendOutTime=0.2}
    boneToModify {boneName=cape_02_r}
    translation {y=3, z=1.5}
   }
   n47 = "Transform (Modify) Bone - Bone: cape_01_l"(ComponentPose=@n46, bAlphaBoolEnabled=@n41, Translation="0,0,0", Rotation="-5,0,0", Scale="1,1,1") {
    alphaInputType=Bool
    rotationMode=BMM_Additive
    rotationSpace=BCS_BoneSpace
    translationMode=BMM_Additive
    translationSpace=BCS_BoneSpace
    alphaBoolBlend {blendInTime=0.2, blendOption=Linear, blendOutTime=0.2}
    boneToModify {boneName=cape_01_l}
    rotation {pitch=-5}
   }
   n48 = "Transform (Modify) Bone - Bone: cape_mid_05_l"!(ComponentPose=@n47, bAlphaBoolEnabled:bool=@n41, Translation="0,0,0", Rotation="10,0,0", Scale="1,1,1") -> (Pose) {
    alphaInputType=Bool
    rotationMode=BMM_Additive
    rotationSpace=BCS_BoneSpace
    translationMode=BMM_Additive
    translationSpace=BCS_BoneSpace
    alphaBoolBlend {blendInTime=0.2, blendOption=Linear, blendOutTime=0.2}
    boneToModify {boneName=cape_mid_05_l}
    rotation {pitch=10}
   }
   n49 = "Transform (Modify) Bone - Bone: cape_mid_05_r"!(ComponentPose=@n48, bAlphaBoolEnabled:bool=@n41, Translation="0,0,0", Rotation="10,0,0", Scale="1,1,1") -> (Pose) {
    alphaInputType=Bool
    rotationMode=BMM_Additive
    rotationSpace=BCS_BoneSpace
    translationMode=BMM_Additive
    translationSpace=BCS_BoneSpace
    alphaBoolBlend {blendInTime=0.2, blendOption=Linear, blendOutTime=0.2}
    boneToModify {boneName=cape_mid_05_r}
    rotation {pitch=10}
   }
   n50 = "Transform (Modify) Bone - Bone: cape_05_l"!(ComponentPose=@n49, bAlphaBoolEnabled:bool=@n41, Translation="0,0,0", Rotation="0,0,0", Scale="1,1,1") -> (Pose) {
    alphaInputType=Bool
    rotationMode=BMM_Additive
    rotationSpace=BCS_BoneSpace
    translationMode=BMM_Additive
    translationSpace=BCS_BoneSpace
    alphaBoolBlend {blendInTime=0.2, blendOption=Linear, blendOutTime=0.2}
    boneToModify {boneName=cape_05_l}
   }
   n51 = "Transform (Modify) Bone - Bone: cape_05_r"!(ComponentPose=@n50, bAlphaBoolEnabled:bool=@n41, Translation="0,0,0", Rotation="10,0,0", Scale="1,1,1") -> (Pose) {
    alphaInputType=Bool
    rotationMode=BMM_Additive
    rotationSpace=BCS_BoneSpace
    translationMode=BMM_Additive
    translationSpace=BCS_BoneSpace
    alphaBoolBlend {blendInTime=0.2, blendOption=Linear, blendOutTime=0.2}
    boneToModify {boneName=cape_05_r}
    rotation {pitch=10}
   }
   n52 = "Transform (Modify) Bone - Bone: cape_01_r"(ComponentPose=@n51, bAlphaBoolEnabled=@n1, Translation="0,7,0", Rotation="0,0,0", Scale="1,1,1") {
    alphaInputType=Bool
    rotationMode=BMM_Additive
    rotationSpace=BCS_BoneSpace
    translationMode=BMM_Additive
    translationSpace=BCS_BoneSpace
    alphaBoolBlend {blendInTime=1, blendOption=Linear}
    boneToModify {boneName=cape_01_r}
    translation {y=7}
   }
   n53 = "Transform (Modify) Bone - Bone: cape_mid_02_l"(ComponentPose=@n52, bAlphaBoolEnabled=@n1, Translation="0,10,0", Rotation="0,0,0", Scale="1,1,1") {
    alphaInputType=Bool
    rotationMode=BMM_Additive
    rotationSpace=BCS_BoneSpace
    translationMode=BMM_Additive
    translationSpace=BCS_BoneSpace
    alphaBoolBlend {blendInTime=0.4, blendOption=Linear, blendOutTime=0.5}
    boneToModify {boneName=cape_mid_02_l}
    translation {y=10}
   }
   n54 = "Transform (Modify) Bone - Bone: cape_02_r"(ComponentPose=@n53, bAlphaBoolEnabled=Get isCrouchMelee() -> (isCrouchMelee:bool), Translation="0,5,0", Rotation="0,5,0", Scale="1,1,1") {
    alphaInputType=Bool
    rotationMode=BMM_Additive
    rotationSpace=BCS_BoneSpace
    translationMode=BMM_Additive
    translationSpace=BCS_BoneSpace
    alphaBoolBlend {blendInTime=0.4, blendOption=Linear, blendOutTime=0.5}
    boneToModify {boneName=cape_02_r}
    rotation {yaw=5}
    translation {y=5}
   }
   n55 = Look At - Bone: dyn_lightSaber_MSTR(ComponentPose=@n54, Alpha=1) {
    alphaBoolBlend {blendOption=Linear}
    boneToModify {boneName=dyn_lightSaber_MSTR}
    lookAtLocation=<-48.7728 -16.917 -9.9033>
    lookAtTarget {boneReference {boneName=thigh_l}}
    lookAt_Axis {axis=<1 0>}
   }
   Output Pose(Result=Blend Poses by bool(BlendPose_0=Use cached pose 'slidingCache'(), BlendPose_1=Component To Local(ComponentPose=@n55), bActiveValue=Get bIsSliding()) {blendType=Linear}) {layerGroup=None}
   n56 = Control Rig(Source=Copy Pose From Mesh() {bUseAttachedParent=true}) {
    alphaBoolBlend {blendOption=Linear}
    controlRigAssetReference {blueprintRigClass=/BRCosmetics/Characters/Player/Male/Medium/Bodies/M_MED_DistantEchoPro_Style/Meshes/M_MED_DeformRig_DistantEchoPro_Style_ControlRig_new.M_MED_DeformRig_DistantEchoPro_Style_ControlRig_new_C}
   }
   Save cached pose 'DEP_StyleCache'(Pose=Component To Local(ComponentPose=Local To Component(LocalPose=@n56))) {cacheName=DEP_StyleCache}
   n57 = Control Rig(Source=Use cached pose 'DEP_StyleCache'()) {
    alphaBoolBlend {blendOption=Linear}
    controlRigAssetReference {blueprintRigClass=&/M_MED_DistantEchoPro_Cloak_CtrlRig.M_MED_DistantEchoPro_Cloak_CtrlRig_C}
   }
   Save cached pose 'cloakCRCache'(Pose=@n57) {cacheName=cloakCRCache}
   n58 = "Constraint - cape_mid_02_r (P)"!(ComponentPose=Local To Component(LocalPose=Use cached pose 'cloakCRCache'()), Alpha:float=1, ConstraintWeights_0=1) -> (Pose) {
    alphaBoolBlend {blendOption=Linear}
    boneToModify {boneName=cape_mid_02_r}
    constraintSetup [{targetBone {boneName=spine_04}}]
    constraintWeights [1]
   }
   n59 = "Constraint - cape_mid_02_l (P)"!(ComponentPose=@n58, Alpha:float=1, ConstraintWeights_0=1) -> (Pose) {
    alphaBoolBlend {blendOption=Linear}
    boneToModify {boneName=cape_mid_02_l}
    constraintSetup [{targetBone {boneName=spine_04}}]
    constraintWeights [1]
   }
   n60 = Get tacticalAccScale() -> (tacticalAccScale:Vector)
   n61 = RigidBody!(
    ComponentPose=@n59
    Alpha
    OverrideWorldGravity=Get Gravity Override()
    ExternalForce
    ComponentLinearAccScale:Vector=@n60
    ComponentLinearVelScale:Vector=@n60
   ) -> (Pose) {
    bClampLinearTranslationLimitToRefPose=true
    bOverrideWorldGravity=true
    overridePhysicsAsset=&/M_MED_DistantEchoPro_Cloak_BackBling_Physics
    simulationSpace=BaseBoneSpace
    alphaBoolBlend {blendOption=Linear}
    baseBoneRef {boneName=root}
    componentAppliedLinearAccClamp=<6000 6000 6000>
    componentLinearAccScale=<0.2 0.2 0.3>
    componentLinearVelScale=<0.2 0.2 0.3>
    overrideWorldGravity {z=-980}
   }
   n62 = Get emote() -> (emote:bool)
   n63 = "Transform (Modify) Bone - Bone: cape_mid_01_l"(ComponentPose=@n61, bAlphaBoolEnabled=@n62, Translation="0,0,2.5", Rotation="12,0,0", Scale="1,1,1") {
    alphaInputType=Bool
    rotationMode=BMM_Additive
    rotationSpace=BCS_BoneSpace
    translationMode=BMM_Additive
    translationSpace=BCS_BoneSpace
    alphaBoolBlend {blendInTime=0.1, blendOption=Linear, blendOutTime=0.1}
    boneToModify {boneName=cape_mid_01_l}
    rotation {pitch=12}
    translation {z=2.5}
   }
   n64 = "Transform (Modify) Bone - Bone: cape_mid_01_r"(ComponentPose=@n63, bAlphaBoolEnabled=@n62, Translation="0,0,2.5", Rotation="12,0,0", Scale="1,1,1") {
    alphaInputType=Bool
    rotationMode=BMM_Additive
    rotationSpace=BCS_BoneSpace
    translationMode=BMM_Additive
    translationSpace=BCS_BoneSpace
    alphaBoolBlend {blendInTime=0.1, blendOption=Linear, blendOutTime=0.1}
    boneToModify {boneName=cape_mid_01_r}
    rotation {pitch=12}
    translation {z=2.5}
   }
   n65 = "Transform (Modify) Bone - Bone: cape_01_l"(ComponentPose=@n64, bAlphaBoolEnabled=@n62, Translation="0,0,0", Rotation="5,-5,0.000", Scale="1,1,1") {
    alphaInputType=Bool
    rotationMode=BMM_Additive
    rotationSpace=BCS_BoneSpace
    alphaBoolBlend {blendInTime=0.1, blendOption=Linear, blendOutTime=0.1}
    boneToModify {boneName=cape_01_l}
    rotation {pitch=5, yaw=-5}
   }
   n66 = "Transform (Modify) Bone - Bone: cape_01_r"(ComponentPose=@n65, bAlphaBoolEnabled=@n62, Translation="0,0,0", Rotation="5,5,0.000", Scale="1,1,1") {
    alphaInputType=Bool
    rotationMode=BMM_Additive
    rotationSpace=BCS_BoneSpace
    alphaBoolBlend {blendInTime=0.1, blendOption=Linear, blendOutTime=0.1}
    boneToModify {boneName=cape_01_r}
    rotation {pitch=5, yaw=5}
   }
   n67 = Get bIsPlayingEmote()
   n68 = "Transform (Modify) Bone - Bone: cape_01_r"(ComponentPose=@n66, bAlphaBoolEnabled=@n67, Translation="0,0,3", Rotation="0,0,0", Scale="1,1,1") {
    alphaInputType=Bool
    rotationMode=BMM_Additive
    rotationSpace=BCS_BoneSpace
    translationMode=BMM_Additive
    translationSpace=BCS_BoneSpace
    alphaBoolBlend {blendInTime=0.1, blendOption=Linear, blendOutTime=0.1}
    boneToModify {boneName=cape_01_r}
    translation {z=3}
   }
   n69 = "Transform (Modify) Bone - Bone: dyn_capeShoulder_r_01"(ComponentPose=@n68, bAlphaBoolEnabled=@n67, Translation="0,2,0", Rotation="0.013,-15.000,0.008", Scale="1,1,1") {
    alphaInputType=Bool
    rotationMode=BMM_Additive
    rotationSpace=BCS_BoneSpace
    translationMode=BMM_Additive
    translationSpace=BCS_BoneSpace
    alphaBoolBlend {blendInTime=0.1, blendOption=Linear, blendOutTime=0.1}
    boneToModify {boneName=dyn_capeShoulder_r_01}
    rotation=<rot 0.013 -15 0.008>
    translation {y=2}
   }
   n70 = "Transform (Modify) Bone - Bone: cape_mid_01_l"(ComponentPose=@n69, bAlphaBoolEnabled=@n67, Translation="0,0,3", Rotation="5,0,0", Scale="1,1,1") {
    alphaInputType=Bool
    rotationMode=BMM_Additive
    rotationSpace=BCS_BoneSpace
    translationMode=BMM_Additive
    translationSpace=BCS_BoneSpace
    alphaBoolBlend {blendInTime=0.1, blendOption=Linear, blendOutTime=0.1}
    boneToModify {boneName=cape_mid_01_l}
    rotation {pitch=5}
    translation {z=3}
   }
   n71 = "Transform (Modify) Bone - Bone: cape_01_l"(ComponentPose=@n70, bAlphaBoolEnabled=@n67, Translation="0,-2,4", Rotation="10,0,0.008", Scale="1,1,1") {
    alphaInputType=Bool
    rotationMode=BMM_Additive
    rotationSpace=BCS_BoneSpace
    translationMode=BMM_Additive
    translationSpace=BCS_BoneSpace
    alphaBoolBlend {blendInTime=0.1, blendOption=Linear, blendOutTime=0.1}
    boneToModify {boneName=cape_01_l}
    rotation {pitch=10, roll=0.008}
    translation {y=-2, z=4}
   }
   n72 = Component To Local(ComponentPose=@n71)
   n73 = Get bIsSkyDiving()
   n74 = Get tacticalAccScale()
   n75 = RigidBody!(
    ComponentPose=Local To Component(LocalPose=Use cached pose 'cloakCRCache'())
    bAlphaBoolEnabled:bool=NOT Boolean(A=@n73)
    OverrideWorldGravity=Get Gravity Override()
    ExternalForce
    ComponentLinearAccScale=@n74
    ComponentLinearVelScale=@n74
   ) -> (Pose) {
    alphaInputType=Bool
    bClampLinearTranslationLimitToRefPose=true
    bOverrideWorldGravity=true
    overridePhysicsAsset=&/M_MED_DistantEchoPro_Cloak_Physics
    simulationSpace=BaseBoneSpace
    alphaBoolBlend {blendInTime=0.1, blendOption=Linear, blendOutTime=0.1}
    baseBoneRef {boneName=root}
    componentAppliedLinearAccClamp=<6000 6000 6000>
    componentLinearAccScale=<0.2 0.4 0.4>
    componentLinearVelScale=<0.2 0.4 0.4>
    overrideWorldGravity {z=-980}
   }
   n76 = Get tacticalAccScale()
   n77 = RigidBody(
    ComponentPose=@n75
    bAlphaBoolEnabled=@n73
    OverrideWorldGravity=Get Gravity Override()
    ComponentLinearAccScale=@n76
    ComponentLinearVelScale=@n76
   ) {
    alphaInputType=Bool
    bClampLinearTranslationLimitToRefPose=true
    bOverrideWorldGravity=true
    overridePhysicsAsset=&/M_MED_DistantEchoPro_Cloak_skydive_Physics1
    simulationSpace=BaseBoneSpace
    alphaBoolBlend {blendInTime=0.1, blendOption=Linear, blendOutTime=0.1}
    baseBoneRef {boneName=root}
    componentAppliedLinearAccClamp=<6000 6000 6000>
    componentLinearAccScale=<0.2 0.4 0.4>
    componentLinearVelScale=<0.2 0.4 0.4>
    overrideWorldGravity {z=-980}
   }
   n78 = Component To Local(ComponentPose=@n77)
   n79 = Get Is Wearing Backbling()
   n80 = Get bIsCrouching()
   n81 = "Constraint - cape_02_l (P)"!(ComponentPose=Local To Component(LocalPose=Use cached pose 'cloakCRCache'()), bAlphaBoolEnabled:bool=@n80, ConstraintWeights_0=1) -> (Pose) {
    alphaInputType=Bool
    alphaBoolBlend {blendOption=Linear}
    boneToModify {boneName=cape_02_l}
    constraintSetup [{targetBone {boneName=spine_05}}]
    constraintWeights [1]
   }
   n82 = "Constraint - cape_01_r (P)"(ComponentPose=@n81, bAlphaBoolEnabled=@n80, ConstraintWeights_0=1) {
    alphaInputType=Bool
    alphaBoolBlend {blendOption=Linear}
    boneToModify {boneName=cape_01_r}
    constraintSetup [{targetBone {boneName=spine_05}}]
    constraintWeights [1]
   }
   n83 = RigidBody!(ComponentPose=@n82, Alpha:float, OverrideWorldGravity=Get Gravity Override(), ExternalForce) -> (Pose) {
    bClampLinearTranslationLimitToRefPose=true
    bOverrideWorldGravity=true
    overridePhysicsAsset=&/M_MED_DistantEchoPro_Cloak_Crouching_Physics
    simulationSpace=BaseBoneSpace
    alphaBoolBlend {blendOption=Linear}
    baseBoneRef {boneName=root}
    componentAppliedLinearAccClamp=<6000 6000 6000>
    componentLinearAccScale=<0.2 0.4 0.6>
    componentLinearVelScale=<0.2 0.4 0.6>
    overrideWorldGravity {z=-980}
   }
   Save cached pose 'simCache'(Pose=Blend Poses by bool(
     BlendPose_0=Component To Local(ComponentPose=@n83)
     BlendPose_1=Blend Poses by bool(BlendPose_0=@n72, BlendPose_1=@n78, bActiveValue=@n79) {blendType=Linear}
     BlendTime_0=1
     bActiveValue=Get isCrouchNotPlayingMelee() -> (isCrouchNotPlayingMelee:bool)
    ) {blendType=Linear}) {cacheName=simCache}
   n84 = "Transform (Modify) Bone - Bone: cape_03_l"(
    ComponentPose=Local To Component(LocalPose=Use cached pose 'DEP_StyleCache'())
    bAlphaBoolEnabled=True
    Translation="0,0,0"
    Rotation="60,0,0"
    Scale="1,1,1"
   ) {
    alphaInputType=Bool
    rotationMode=BMM_Additive
    rotationSpace=BCS_BoneSpace
    alphaBoolBlend {blendInTime=0.2, blendOption=Linear, blendOutTime=0.2}
    boneToModify {boneName=cape_03_l}
    rotation {pitch=60}
   }
   n85 = "Transform (Modify) Bone - Bone: cape_05_l"(ComponentPose=@n84, bAlphaBoolEnabled=True, Translation="0,0,0", Rotation="20,0,0", Scale="1,1,1") {
    alphaInputType=Bool
    rotationMode=BMM_Additive
    rotationSpace=BCS_BoneSpace
    alphaBoolBlend {blendInTime=0.2, blendOption=Linear, blendOutTime=0.2}
    boneToModify {boneName=cape_05_l}
    rotation {pitch=20}
   }
   n86 = "Transform (Modify) Bone - Bone: cape_06_l"!(ComponentPose=@n85, bAlphaBoolEnabled:bool=True, Translation="0,0,0", Rotation="10,0,0", Scale="1,1,1") -> (Pose) {
    alphaInputType=Bool
    rotationMode=BMM_Additive
    rotationSpace=BCS_BoneSpace
    alphaBoolBlend {blendInTime=0.2, blendOption=Linear, blendOutTime=0.2}
    boneToModify {boneName=cape_06_l}
    rotation {pitch=10}
   }
   n87 = "Transform (Modify) Bone - Bone: cape_mid_03_l"(ComponentPose=@n86, bAlphaBoolEnabled=True, Translation="0,0,0", Rotation="60,0,0", Scale="1,1,1") {
    alphaInputType=Bool
    rotationMode=BMM_Additive
    rotationSpace=BCS_BoneSpace
    alphaBoolBlend {blendInTime=0.2, blendOption=Linear, blendOutTime=0.2}
    boneToModify {boneName=cape_mid_03_l}
    rotation {pitch=60}
   }
   n88 = "Transform (Modify) Bone - Bone: cape_mid_05_l"(ComponentPose=@n87, bAlphaBoolEnabled=True, Translation="0,0,0", Rotation="20,0,0", Scale="1,1,1") {
    alphaInputType=Bool
    rotationMode=BMM_Additive
    rotationSpace=BCS_BoneSpace
    alphaBoolBlend {blendInTime=0.2, blendOption=Linear, blendOutTime=0.2}
    boneToModify {boneName=cape_mid_05_l}
    rotation {pitch=20}
   }
   n89 = "Transform (Modify) Bone - Bone: cape_mid_06_l"!(ComponentPose=@n88, bAlphaBoolEnabled:bool=True, Translation="0,0,0", Rotation="10,0,0", Scale="1,1,1") -> (Pose) {
    alphaInputType=Bool
    rotationMode=BMM_Additive
    rotationSpace=BCS_BoneSpace
    alphaBoolBlend {blendInTime=0.2, blendOption=Linear, blendOutTime=0.2}
    boneToModify {boneName=cape_mid_06_l}
    rotation {pitch=10}
   }
   n90 = "Transform (Modify) Bone - Bone: cape_mid_03_r"(ComponentPose=@n89, bAlphaBoolEnabled=True, Translation="0,0,0", Rotation="60,0,0", Scale="1,1,1") {
    alphaInputType=Bool
    rotationMode=BMM_Additive
    rotationSpace=BCS_BoneSpace
    alphaBoolBlend {blendInTime=0.2, blendOption=Linear, blendOutTime=0.2}
    boneToModify {boneName=cape_mid_03_r}
    rotation {pitch=60}
   }
   n91 = "Transform (Modify) Bone - Bone: cape_mid_05_r"(ComponentPose=@n90, bAlphaBoolEnabled=True, Translation="0,0,0", Rotation="20,0,0", Scale="1,1,1") {
    alphaInputType=Bool
    rotationMode=BMM_Additive
    rotationSpace=BCS_BoneSpace
    alphaBoolBlend {blendInTime=0.2, blendOption=Linear, blendOutTime=0.2}
    boneToModify {boneName=cape_mid_05_r}
    rotation {pitch=20}
   }
   n92 = "Transform (Modify) Bone - Bone: cape_mid_06_r"!(ComponentPose=@n91, bAlphaBoolEnabled:bool=True, Translation="0,0,0", Rotation="10,0,0", Scale="1,1,1") -> (Pose) {
    alphaInputType=Bool
    rotationMode=BMM_Additive
    rotationSpace=BCS_BoneSpace
    alphaBoolBlend {blendInTime=0.2, blendOption=Linear, blendOutTime=0.2}
    boneToModify {boneName=cape_mid_06_r}
    rotation {pitch=10}
   }
   n93 = "Transform (Modify) Bone - Bone: cape_03_r"!(ComponentPose=@n92, bAlphaBoolEnabled:bool=True, Translation="0,0,0", Rotation="45,0,0", Scale="1,1,1") -> (Pose) {
    alphaInputType=Bool
    rotationMode=BMM_Additive
    rotationSpace=BCS_BoneSpace
    alphaBoolBlend {blendInTime=0.2, blendOption=Linear, blendOutTime=0.2}
    boneToModify {boneName=cape_03_r}
    rotation {pitch=45}
   }
   n94 = "Transform (Modify) Bone - Bone: cape_05_r"(ComponentPose=@n93, bAlphaBoolEnabled=True, Translation="0,0,0", Rotation="20,0,0", Scale="1,1,1") {
    alphaInputType=Bool
    rotationMode=BMM_Additive
    rotationSpace=BCS_BoneSpace
    alphaBoolBlend {blendInTime=0.2, blendOption=Linear, blendOutTime=0.2}
    boneToModify {boneName=cape_05_r}
    rotation {pitch=20}
   }
   n95 = "Transform (Modify) Bone - Bone: cape_06_r"!(ComponentPose=@n94, bAlphaBoolEnabled:bool=True, Translation="0,0,0", Rotation="10,0,0", Scale="1,1,1") -> (Pose) {
    alphaInputType=Bool
    rotationMode=BMM_Additive
    rotationSpace=BCS_BoneSpace
    alphaBoolBlend {blendInTime=0.2, blendOption=Linear, blendOutTime=0.2}
    boneToModify {boneName=cape_06_r}
    rotation {pitch=10}
   }
   n96 = RigidBody!(ComponentPose=@n95, Alpha, ExternalForce) -> (Pose) {
    bClampLinearTranslationLimitToRefPose=true
    bOverrideWorldGravity=true
    overridePhysicsAsset=&/M_MED_DistantEchoPro_Cloak_Sliding_Physics
    simulationSpace=BaseBoneSpace
    alphaBoolBlend {blendOption=Linear}
    baseBoneRef {boneName=root}
    componentAppliedLinearAccClamp=<6000 6000 6000>
    componentLinearAccScale=<0.2 0.2 0.2>
    componentLinearVelScale=<0.2 0.2 0.2>
    overrideWorldGravity {z=-980}
   }
   Save cached pose 'slidingCache'(Pose=Component To Local(ComponentPose=@n96)) {cacheName=slidingCache}
   Use cached pose 'cloakCRCache'()
   Get bIsMoving2D()
   Get emote()
   Get isWalking() -> (isWalking:bool)
   Get isCrouchWalking()
   Get isCrouchMelee()
  ]}
 EventGraph {nodes [
   n1 = Event BlueprintUpdateAnimation()
   > Set Delta Time X(Delta Time X=@n1.DeltaTimeX)
   > n2 = Sequence!(execute) -> (then_0, then_1, then_2:Exec, then_3:Exec, then_4:Exec, then_5:Exec, then_6:Exec, then_7:Exec, then_8:Exec, then_9:Exec)
   > n3 = Gravity Override State Function DEPRECATED(CharacterPartAnimInstance=Self-Reference(), GameTimeInSeconds=Get Delta Time X(), GravityOverrideParams=Get Gravity Override Params())
   > Set Gravity Override(Gravity Override=@n3)
   Set bodyMesh(execute=@n2.then_1, bodyMesh:Skeletal Mesh Component Object Reference=GetOwnerSkeletalMeshComponent()) -> (Output_Get:Skeletal Mesh Component Object Reference)
   Set Raw Output(
    execute=@n2.then_2
    Raw Output:float=GetAlphaFromReferenceTransformDelta(
     SkeletalMeshComponent=Get bodyMesh() -> (bodyMesh:Skeletal Mesh Component Object Reference)
     BoneName=upperarm_r
     BaseName=clavicle_r
     InputRange=(X=0,Y=30)
     TwistAxis="1,0,0"
    ).RawOutput
   ) -> (Output_Get:float)
   Set Is Wearing Backbling(execute=@n2.then_3, Is Wearing Backbling=CheckForBackbling())
   n4 = Get bIsPlayingEmote()
   > Set Front_Emote(Front_Emote:bool=AND Boolean(A=NOT Boolean(A=@n4), B=Get bIsInFrontEnd())) -> (Output_Get:bool)
   n5 = Cast To FortPlayerAnimInstance(execute=@n2.then_4, Object=GetAnimInstance(self=GetOwnerSkeletalMeshComponent()))
   > Set isWeaponMelee(isWeaponMelee:bool=false) -> (Output_Get:bool)
   > Switch on EFortWeaponCoreAnimation!(execute, Selection:EFortWeaponCoreAnimation Enum=Get WeaponCoreAnim(self=@n5)) -> (
    Melee:Exec
    Pistol:Exec
    Shotgun:Exec
    PaperBlueprint:Exec
    Rifle:Exec
    MeleeOneHand:Exec
    MachinePistol:Exec
    RocketLauncher:Exec
    GrenadeLauncher:Exec
    GoingCommando:Exec
    AssaultRifle:Exec
    TacticalShotgun:Exec
    SniperRifle:Exec
    TrapPlacement:Exec
    ShoulderLauncher:Exec
    AbilityDecoTool:Exec
    Crossbow:Exec
    C4:Exec
    RemoteControl:Exec
    DualWield:Exec
    AR_BullPup:Exec
    AR_ForwardGrip:Exec
    MedPackPaddles:Exec
    SMG_P90:Exec
    AR_DrumGun:Exec
    Consumable_Small:Exec
    Consumable_Large:Exec
    Balloon:Exec
    MountedTurret:Exec
    CreativeTool:Exec
    ExplosiveBow:Exec
    AshtonIndigo:Exec
    AshtonChicago:Exec
    MeleeDualWield:Exec
    Unarmed:Exec
   )
   > n6 = Set isWeaponMelee(isWeaponMelee=true)
   > n7 = Branch(Condition=AND Boolean(A=@n6, B=Get bIsCrouching()))
   > Set isCrouchMelee(isCrouchMelee:bool=true) -> (Output_Get:bool)
   Set isCrouchMelee(execute=@n7.else, isCrouchMelee=false)
   n8 = Get bIsPlayingMeleeAnim()
   n9 = Branch(execute=@n2.then_5, Condition=AND Boolean(A=NOT Boolean(A=@n8), B=Get bIsCrouching()))
   > Set isCrouchNotPlayingMelee(isCrouchNotPlayingMelee:bool=true) -> (Output_Get:bool)
   Set isCrouchNotPlayingMelee(execute=@n9.else, isCrouchNotPlayingMelee=false)
   n10 = Branch(execute=@n2.then_6, Condition=Get bIsTacticalSprinting())
   > Set tacticalAccScale(tacticalAccScale:Vector="0.01,0.01,0.02") -> (Output_Get:Vector)
   Set tacticalAccScale(execute=@n10.else, tacticalAccScale="0.2,0.4,0.4")
   n11 = PlayerMovement(execute=@n2.then_7, AnimInstance=Self-Reference(), XDotProductCoefficient=1, YDotProductCoefficient=1, ZDotProductCoefficient=1)
   > n12 = Sequence!(execute) -> (then_0, then_1, then_2)
   > n13 = Branch(Condition=OR Boolean!(A="Not Equal (Float)"(A=@n11), B="Not Equal (Float)"(A=@n11.YMovement), C:bool="Not Equal (Float)"(A=@n11.ZMovement)) -> (ReturnValue))
   > Set isWalking(isWalking:bool=true) -> (Output_Get:bool)
   Set isWalking(execute=@n13.else, isWalking=false)
   n14 = "float > float"(A=MapRangeClamped(Value=@n11, InRangeA=-1, InRangeB=0, OutRangeA=1, OutRangeB=0), B=0.5)
   n15 = MapRangeClamped(Value=@n11, InRangeA=0, InRangeB=1)
   n16 = Branch(execute=@n12.then_1, Condition=OR Boolean!(A="float > float"(A=@n15, B=0.5), B=@n14) -> (ReturnValue))
   > Set PlayerStrafe(PlayerStrafe:bool=true) -> (Output_Get:bool)
   Set PlayerStrafe(execute=@n16.else, PlayerStrafe=false)
   n17 = Branch(execute=@n12.then_2, Condition=AND Boolean(A=@n14, B=Get bIsCrouching()))
   > Set isCrouchStrafeLeft(isCrouchStrafeLeft:bool=true) -> (Output_Get:bool)
   Set isCrouchStrafeLeft(execute=@n17.else, isCrouchStrafeLeft=false)
   n18 = Branch(execute=@n2.then_8, Condition=AND Boolean(A=Get bIsCrouching(), B=Get isWalking()))
   > Set isCrouchWalking(isCrouchWalking:bool=true) -> (Output_Get:bool)
   Set isCrouchWalking(execute=@n18.else, isCrouchWalking=false)
   n19 = GetAnimInstance(self=GetOwnerSkeletalMeshComponent())
   Is Valid(exec=@n2.then_9, InputObject=@n19)
   > Set emote(emote:bool=OR Boolean!(
     A=Montage_IsPlaying(self=@n19, Montage=/BRCosmetics/Animation/Game/MainPlayer/Emotes/Noodles/Emote_Noodles_CMM_M)
     B=Montage_IsPlaying(self=@n19, Montage=/BRCosmetics/Animation/Game/MainPlayer/Emotes/Temple/Emote_Temple_CMM_M)
     C:bool=Montage_IsPlaying(self=@n19, Montage=/BRCosmetics/Animation/Game/MainPlayer/Emotes/BlackMondayFemale/emote_blackmondayfemale_CMM_Montage)
     D:bool=Montage_IsPlaying(self=@n19, Montage=/BRCosmetics/Animation/Game/MainPlayer/Emotes/KeeperDreamChorus/Emote_KeeperDreamChorus_CMM_M)
     E:bool=Montage_IsPlaying(self=@n19, Montage=/BRCosmetics/Animation/Game/MainPlayer/Emotes/Cyclone/Cyclone_CMM_M)
     F:bool=Montage_IsPlaying(self=@n19, Montage=/BRCosmetics/Animation/Game/MainPlayer/Emotes/KeeperDreamHook/Emote_KeeperDreamHook_CMM_M)
    ) -> (ReturnValue)) -> (Output_Get:bool)
   GetPartSkeletalMeshComponent()
  ]}
]
variables [
 ar_firingPosition {type=integer}
 Binding Context {
  type=ContextualAnimSceneBindingContext
  default {
   actor=null
   cachedAnimInstance=null
   cachedMotionWarpingComp=null
   cachedMovementComp=null
   cachedSceneActorComp=null
   cachedSkeletalMesh=null
  }
 }
 bodyMesh {type=SkeletalMeshComponent}
 Delta Time X {type=number}
 emote {type=boolean}
 Front_Emote {type=boolean}
 Gravity Override {type=Vector}
 Gravity Override Params {
  type=GravityOverrideParamsStruct
  default {
   skydive_DivingDown {jointName=root, gravityOverride=<0 -1000 0>, windAmplitude=<50 30 100>}
   skydive_FloatingDown {gravityOverride=<0 -700 1000>}
  }
 }
 Is Wearing Backbling {type=boolean}
 isCrouchMelee {type=boolean}
 isCrouchNotPlayingMelee {type=boolean}
 isCrouchStrafeLeft {type=boolean}
 isCrouchWalking {type=boolean}
 isWalking {type=boolean}
 isWeaponMelee {type=boolean}
 NewVar {type=number}
 Pawn Movement Direction Azimuth {type=number}
 Pawn Velocity {type=Vector}
 PlayerStrafe {type=boolean}
 Raw Output {type=number}
 tacticalAccScale {type=Vector}
 Velocity Lerp Alpha {type=number}
]
