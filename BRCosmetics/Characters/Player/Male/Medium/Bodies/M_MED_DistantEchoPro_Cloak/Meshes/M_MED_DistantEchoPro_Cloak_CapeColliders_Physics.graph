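# Physics asset for the cloak of the male medium DistantEchoPro body, giving the cape bones and nearby limbs capsule collision shapes.
# Four cape chains (left, right, mid-left, mid-right) each link seven bones to a small MSTR root body through swing-limited, zero-twist constraints.
# Swing limits are tighter near the roots, except on the uniform outer-left chain.
# Arm and leg capsules act as body colliders, with free constraints at elbows and knees.
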PhysicsAsset /BRCosmetics/Characters/Player/Male/Medium/Bodies/M_MED_DistantEchoPro_Cloak/Meshes/M_MED_DistantEchoPro_Cloak_CapeColliders_Physics
bodies {
 calf_l [{shapeType=Capsule, length=34.8147, radius=8.5869, shapeName=calf_l_capsule, center=<-21.1879 0 0>, rotation=<rot 90 0 0>}]
 calf_r [{shapeType=Capsule, length=34.8147, radius=8.5869, shapeName=calf_l_capsule, center=<21.1879 0 0>, rotation=<rot 90 0 180>}]
 cape_01_l [{shapeType=Capsule, length=15.2959, radius=4.0644, shapeName=cape_01_l_capsule, center=<9.0604 0 -2.6822>, rotation=<rot 90 0 0>}]
 cape_01_r [
  {shapeType=Capsule, length=19.1706, radius=4.9191, shapeName=cape_01_r_capsule, center=<9.0604 0 -3.636>, rotation=<rot 90 0 0>}
  {shapeType=Capsule, length=17.7358, radius=5.7925, shapeName=cape_01_r_capsule, center=<14.9744 0 -4.2675>, rotation=<rot 0 0 90>}
 ]
 cape_02_l [{shapeType=Capsule, length=22.8431, radius=4.9598, shapeName=cape_02_l_capsule, center=<0 0 -3.7512>, rotation=<rot 90 -165.9638 104.0362>}]
 cape_02_r [{
   shapeType=Capsule
   length=21.4566
   radius=4.0305
   shapeName=cape_02_r_capsule
   center=<9.0604 0.4164 1.8>
   rotation=<rot 84.7003 -90.0001 -179.9998>
  }]
 cape_03_l [{shapeType=Capsule, length=22.5906, radius=9.2211, shapeName=cape_03_l_capsule, center=<0 0 -3.4984>, rotation=<rot 90 0 0>}]
 cape_03_r [{shapeType=Capsule, length=22.5906, radius=9.2211, shapeName=cape_03_r_capsule, center=<0 0 -4.0615>, rotation=<rot 90 0 0>}]
 cape_04_l [{shapeType=Capsule, length=14.313, radius=9.5779, shapeName=cape_04_l_capsule, center=<7.6186 0 -2.1566>, rotation=<rot 0 0 90>}]
 cape_04_r [{shapeType=Capsule, length=14.313, radius=9.5779, shapeName=cape_04_r_capsule, center=<7.6186 0 -2.9419>, rotation=<rot 0 0 90>}]
 cape_05_l [{shapeType=Capsule, length=16.4461, radius=8.4269, shapeName=cape_05_l_capsule, center=<0 -0.3712 -2.5969>, rotation=<rot -0.0002 0 98.1346>}]
 cape_05_r [{shapeType=Capsule, length=16.4461, radius=8.4269, shapeName=cape_05_l_capsule, center=<0 0.2724 -2.748>, rotation=<rot -0.0003 0 84.3395>}]
 cape_06_l [{shapeType=Capsule, length=19.7585, radius=9.2204, shapeName=cape_06_l_capsule, center=<0 -0.6801 -3.6625>, rotation=<rot -0.0003 0 100.5198>}]
 cape_06_r [{shapeType=Capsule, length=19.7585, radius=9.2204, shapeName=cape_06_l_capsule, center=<0 0.7353 -3.6518>, rotation=<rot -0.0005 0 78.6152>}]
 cape_07_l [{shapeType=Capsule, length=20.4703, radius=7.4456, shapeName=cape_07_l_capsule, rotation=<rot -0.0001 0 99.2564>}]
 cape_07_r [{shapeType=Capsule, length=20.4703, radius=7.4456, shapeName=cape_07_l_capsule, rotation=<rot -0.0002 0 82.4145>}]
 cape_l_MSTR [{shapeType=Capsule, length=1, radius=1, shapeName=cape_l_MSTR_capsule}]
 cape_mid_01_l [{shapeType=Capsule, length=14.5642, radius=4.7961, shapeName=cape_mid_01_l_capsule, center=<11.341 0 0>, rotation=<rot 90 0 0>}]
 cape_mid_01_r [{shapeType=Capsule, length=14.5642, radius=4.7961, shapeName=cape_mid_01_r_capsule, center=<11.341 0 0>, rotation=<rot 90 0 0>}]
 cape_mid_02_l [{shapeType=Capsule, length=17.5485, radius=9.2211, shapeName=cape_mid_02_l_capsule, center=<9.8079 0 -2.9922>, rotation=<rot 90 0 0>}]
 cape_mid_02_r [{shapeType=Capsule, length=17.5485, radius=9.2211, shapeName=cape_mid_02_r_capsule, center=<9.8079 0 -2.9922>, rotation=<rot 90 0 0>}]
 cape_mid_03_l [{
   shapeType=Capsule
   length=18.4933
   radius=9.9972
   shapeName=cape_mid_03_l_capsule
   center=<13.2541 0.0164 -5.6168>
   rotation=<rot 90 -172.875 -82.9461>
  }]
 cape_mid_03_r [{
   shapeType=Capsule
   length=18.4933
   radius=9.9972
   shapeName=cape_mid_03_r_capsule
   center=<13.2541 -0.02 -5.6168>
   rotation=<rot 90 -172.875 -82.7886>
  }]
 cape_mid_04_l [{shapeType=Capsule, length=14.313, radius=9.5779, shapeName=cape_mid_04_l_capsule, center=<7.6186 0 -3.7251>, rotation=<rot 0 0 90>}]
 cape_mid_04_r [{shapeType=Capsule, length=14.313, radius=9.5779, shapeName=cape_mid_04_r_capsule, center=<7.6186 0 -3.7251>, rotation=<rot 0 0 90>}]
 cape_mid_05_l [{shapeType=Capsule, length=19.3687, radius=8.4269, shapeName=cape_mid_05_l_capsule, center=<0 0 -3.7251>, rotation=<rot 0 0 90>}]
 cape_mid_05_r [{shapeType=Capsule, length=19.3687, radius=8.4269, shapeName=cape_mid_05_r_capsule, center=<0 0 -3.7251>, rotation=<rot 0 0 90>}]
 cape_mid_06_l [{shapeType=Capsule, length=21.0375, radius=8.4269, shapeName=cape_mid_06_l_capsule, center=<0 0 -3.7251>, rotation=<rot 0 0 90>}]
 cape_mid_06_r [{shapeType=Capsule, length=21.0375, radius=8.4269, shapeName=cape_mid_06_r_capsule, center=<0 0 -3.7251>, rotation=<rot 0 0 90>}]
 cape_mid_07_l [{shapeType=Capsule, length=20.8926, radius=8.1896, shapeName=cape_mid_07_l_capsule, rotation=<rot 0 0 90>}]
 cape_mid_07_r [{shapeType=Capsule, length=20.8926, radius=8.1896, shapeName=cape_mid_07_r_capsule, rotation=<rot 0 0 90>}]
 cape_mid_l_MSTR [{shapeType=Capsule, length=1, radius=1, shapeName=cape_mid_l_MSTR_capsule}]
 cape_mid_r_MSTR [{shapeType=Capsule, length=1, radius=1, shapeName=cape_mid_r_MSTR_capsule}]
 cape_r_MSTR [{shapeType=Capsule, length=1, radius=1, shapeName=cape_r_MSTR_capsule}]
 lowerarm_l [{shapeType=Capsule, length=21.1522, radius=5.5946, shapeName=lowerarm_l_capsule, center=<10.5383 0 0>, rotation=<rot 90 0 0>}]
 lowerarm_r [{shapeType=Capsule, length=21.1522, radius=5.5946, shapeName=lowerarm_l_capsule, center=<-10.5383 0 0>, rotation=<rot 90 0 180>}]
 thigh_l [{shapeType=Capsule, length=24.3818, radius=11.6052, shapeName=thigh_l_capsule, center=<-21.1879 0 0>, rotation=<rot 90 0 0>}]
 thigh_r [{shapeType=Capsule, length=24.3818, radius=11.6052, shapeName=thigh_l_capsule, center=<21.1879 0 0>, rotation=<rot 90 0 180>}]
 upperarm_l [{shapeType=Capsule, length=20.4503, radius=6.1564, shapeName=upperarm_l_capsule, center=<11.2362 0 0>, rotation=<rot 90 0 0>}]
 upperarm_r [{shapeType=Capsule, length=20.4503, radius=6.1564, shapeName=upperarm_l_capsule, center=<-11.2362 0 0>, rotation=<rot 90 0 180>}]
}
constraints [
 {bone1Name=cape_mid_01_l, bone2Name=cape_mid_l_MSTR, swing1LimitDegrees=15, swing2LimitDegrees=15, twistLimitDegrees=0}
 {bone1Name=cape_mid_02_l, bone2Name=cape_mid_01_l, swing1LimitDegrees=12, swing2LimitDegrees=30, twistLimitDegrees=0}
 {bone1Name=cape_mid_03_l, bone2Name=cape_mid_02_l, swing1LimitDegrees=18, swing2LimitDegrees=30, twistLimitDegrees=0}
 {bone1Name=cape_mid_04_l, bone2Name=cape_mid_03_l, swing1LimitDegrees=30, swing2LimitDegrees=30, twistLimitDegrees=0}
 {bone1Name=cape_mid_05_l, bone2Name=cape_mid_04_l, swing1LimitDegrees=30, swing2LimitDegrees=30, twistLimitDegrees=0}
 {bone1Name=cape_mid_06_l, bone2Name=cape_mid_05_l, swing1LimitDegrees=30, swing2LimitDegrees=30, twistLimitDegrees=0}
 {bone1Name=cape_mid_07_l, bone2Name=cape_mid_06_l, swing1LimitDegrees=30, swing2LimitDegrees=30, twistLimitDegrees=0}
 {bone1Name=cape_mid_01_r, bone2Name=cape_mid_r_MSTR, swing1LimitDegrees=15, swing2LimitDegrees=15, twistLimitDegrees=0}
 {bone1Name=cape_mid_02_r, bone2Name=cape_mid_01_r, swing1LimitDegrees=12, swing2LimitDegrees=30, twistLimitDegrees=0}
 {bone1Name=cape_mid_03_r, bone2Name=cape_mid_02_r, swing1LimitDegrees=18, swing2LimitDegrees=30, twistLimitDegrees=0}
 {bone1Name=cape_mid_04_r, bone2Name=cape_mid_03_r, swing1LimitDegrees=30, swing2LimitDegrees=30, twistLimitDegrees=0}
 {bone1Name=cape_mid_05_r, bone2Name=cape_mid_04_r, swing1LimitDegrees=30, swing2LimitDegrees=30, twistLimitDegrees=0}
 {bone1Name=cape_mid_06_r, bone2Name=cape_mid_05_r, swing1LimitDegrees=30, swing2LimitDegrees=30, twistLimitDegrees=0}
 {bone1Name=cape_mid_07_r, bone2Name=cape_mid_06_r, swing1LimitDegrees=30, swing2LimitDegrees=30, twistLimitDegrees=0}
 {bone1Name=cape_01_l, bone2Name=cape_l_MSTR, swing1LimitDegrees=30, swing2LimitDegrees=30, twistLimitDegrees=0}
 {bone1Name=cape_02_l, bone2Name=cape_01_l, swing1LimitDegrees=30, swing2LimitDegrees=30, twistLimitDegrees=0}
 {bone1Name=cape_03_l, bone2Name=cape_02_l, swing1LimitDegrees=30, swing2LimitDegrees=30, twistLimitDegrees=0}
 {bone1Name=cape_04_l, bone2Name=cape_03_l, swing1LimitDegrees=30, swing2LimitDegrees=30, twistLimitDegrees=0}
 {bone1Name=cape_05_l, bone2Name=cape_04_l, swing1LimitDegrees=30, swing2LimitDegrees=30, twistLimitDegrees=0}
 {bone1Name=cape_06_l, bone2Name=cape_05_l, swing1LimitDegrees=30, swing2LimitDegrees=30, twistLimitDegrees=0}
 {bone1Name=cape_07_l, bone2Name=cape_06_l, swing1LimitDegrees=30, swing2LimitDegrees=30, twistLimitDegrees=0}
 {
  bone1Name=lowerarm_l
  bone2Name=upperarm_l
  swing1LimitDegrees=45
  swing1Motion=Free
  swing2LimitDegrees=45
  swing2Motion=Free
  twistLimitDegrees=45
  twistMotion=Free
 }
 {bone1Name=cape_01_r, bone2Name=cape_r_MSTR, swing1LimitDegrees=15, swing2LimitDegrees=30, twistLimitDegrees=0}
 {bone1Name=cape_02_r, bone2Name=cape_01_r, swing1LimitDegrees=15, swing2LimitDegrees=20, twistLimitDegrees=0}
 {bone1Name=cape_03_r, bone2Name=cape_02_r, swing1LimitDegrees=15, swing2LimitDegrees=20, twistLimitDegrees=0}
 {bone1Name=cape_04_r, bone2Name=cape_03_r, swing1LimitDegrees=30, swing2LimitDegrees=30, twistLimitDegrees=0}
 {bone1Name=cape_05_r, bone2Name=cape_04_r, swing1LimitDegrees=30, swing2LimitDegrees=30, twistLimitDegrees=0}
 {bone1Name=cape_06_r, bone2Name=cape_05_r, swing1LimitDegrees=30, swing2LimitDegrees=30, twistLimitDegrees=0}
 {bone1Name=cape_07_r, bone2Name=cape_06_r, swing1LimitDegrees=30, swing2LimitDegrees=30, twistLimitDegrees=0}
 {
  bone1Name=calf_l
  bone2Name=thigh_l
  swing1LimitDegrees=45
  swing1Motion=Free
  swing2LimitDegrees=45
  swing2Motion=Free
  twistLimitDegrees=45
  twistMotion=Free
 }
 {
  bone1Name=calf_r
  bone2Name=thigh_r
  swing1LimitDegrees=45
  swing1Motion=Free
  swing2LimitDegrees=45
  swing2Motion=Free
  twistLimitDegrees=45
  twistMotion=Free
 }
 {
  bone1Name=lowerarm_r
  bone2Name=upperarm_r
  swing1LimitDegrees=45
  swing1Motion=Free
  swing2LimitDegrees=45
  swing2Motion=Free
  twistLimitDegrees=45
  twistMotion=Free
 }
]
