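# Physics asset for the DistantEchoPro cloak's face accessory part on the male medium player body, simulating the back of the hood.
# Two hood chains, A and B, each pair a tiny master capsule, angularly free relative to spine_05, with a larger child capsule that has limited swing and zero twist.
# Capsules on head and upper spine, joined by their own constraint, give the hood body volumes to work against.
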PhysicsAsset /BRCosmetics/Characters/Player/Male/Medium/Bodies/M_MED_DistantEchoPro_Cloak/Meshes/Parts/M_MED_DistantEchoPro_Cloak_FaceAcc_Physics
bodies {
 dyn_hoodBk_A_01 [{shapeType=Capsule, length=0.342, radius=5.3809, shapeName=dyn_hoodBk_A_01_capsule, center=<5.7521 0 0>, rotation=<rot 90 0 0>}]
 dyn_hoodBk_A_MSTR [{shapeType=Capsule, length=0.505, radius=0.505, shapeName=dyn_hoodBk_A_MSTR_capsule, rotation=<rot 0 -90 0>}]
 dyn_hoodBk_B_01 [{shapeType=Capsule, length=5.8286, radius=2.329, shapeName=dyn_hoodBk_B_01_capsule, center=<4.3125 0.0753 0>, rotation=<rot 0 91 90>}]
 dyn_hoodBk_B_MSTR [{shapeType=Capsule, length=0.505, radius=0.505, shapeName=dyn_hoodBk_B_MSTR_capsule, rotation=<rot 0 -90 0>}]
 head [{shapeType=Capsule, length=5, radius=5, rotation=<rot 90 0 0>}]
 spine_05 [{shapeType=Capsule, length=19.1299, radius=15.6625}]
}
constraints [
 {bone1Name=head, bone2Name=spine_05, swing1LimitDegrees=45, swing2LimitDegrees=45, twistLimitDegrees=45}
 {
  bone1Name=dyn_hoodBk_A_MSTR
  bone2Name=spine_05
  swing1LimitDegrees=45
  swing1Motion=Free
  swing2LimitDegrees=45
  swing2Motion=Free
  twistLimitDegrees=45
  twistMotion=Free
 }
 {bone1Name=dyn_hoodBk_A_01, bone2Name=dyn_hoodBk_A_MSTR, swing1LimitDegrees=20, swing2LimitDegrees=20, twistLimitDegrees=0}
 {
  bone1Name=dyn_hoodBk_B_MSTR
  bone2Name=spine_05
  swing1LimitDegrees=45
  swing1Motion=Free
  swing2LimitDegrees=45
  swing2Motion=Free
  twistLimitDegrees=45
  twistMotion=Free
 }
 {bone1Name=dyn_hoodBk_B_01, bone2Name=dyn_hoodBk_B_MSTR, swing1LimitDegrees=25, swing2LimitDegrees=25, twistLimitDegrees=0}
]
