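# Post-processes the DistantEchoPro cloak's face accessory for the medium male player, adding physics and head-driven corrective motion.
# AnimGraph copies the attached parent's pose, simulates RigidBody relative to spine_05, then applies additive bone-space offsets to hood and tentacle bones.
# EventGraph derives those offsets' up/down alphas from the body mesh's head-to-neck_01 rotation.
# It also feeds the RigidBody gravity override through a deprecated helper function.
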AnimBlueprint /BRCosmetics/Characters/Player/Male/Medium/Bodies/M_MED_DistantEchoPro_Cloak/Meshes/Parts/M_MED_DistantEchoPro_Cloak_FaceAcc_AnimBP
target_skeleton=&/M_MED_DistantEchoPro_Cloak_FaceAcc_Skeleton
graphs [
 AnimGraph {nodes [
   n1 = RigidBody!(
    ComponentPose=Local To Component(LocalPose=Copy Pose From Mesh() {bUseAttachedParent=true})
    Alpha
    OverrideWorldGravity:Vector=Get Gravity Override()
    ExternalForce
   ) -> (Pose) {
    bClampLinearTranslationLimitToRefPose=true
    bOverrideWorldGravity=true
    overridePhysicsAsset=&/M_MED_DistantEchoPro_Cloak_FaceAcc_Physics
    simulationSpace=BaseBoneSpace
    alphaBoolBlend {blendOption=Linear}
    baseBoneRef {boneName=spine_05}
    componentAppliedLinearAccClamp=<6000 6000 6000>
    componentLinearAccScale=<0.6 1 0.6>
    componentLinearVelScale=<0.6 1 0.6>
    overrideWorldGravity {z=-980}
   }
   n2 = "Transform (Modify) Bone - Bone: dyn_hoodBk_A_MSTR"(ComponentPose=@n1, Alpha=Get head Rot Up() -> (head Rot Up:float), Translation="2.2,-2.2,-0.7", Rotation="23,-18,0", Scale="1,1,1") {
    rotationMode=BMM_Additive
    rotationSpace=BCS_BoneSpace
    translationMode=BMM_Additive
    translationSpace=BCS_BoneSpace
    alphaBoolBlend {blendOption=Linear}
    boneToModify {boneName=dyn_hoodBk_A_MSTR}
    rotation {pitch=23, yaw=-18}
    translation=<2.2 -2.2 -0.7>
   }
   n3 = Get head Rot Up()
   n4 = "Transform (Modify) Bone - Bone: dyn_tentacle_2"(ComponentPose=@n2, Alpha=@n3, Translation="-4,0,0", Rotation="-25,0,0", Scale="1,1,1") {
    rotationMode=BMM_Additive
    rotationSpace=BCS_BoneSpace
    translationMode=BMM_Additive
    translationSpace=BCS_BoneSpace
    alphaBoolBlend {blendOption=Linear}
    boneToModify {boneName=dyn_tentacle_2}
    rotation {pitch=-25}
    translation {x=-4}
   }
   n5 = "Transform (Modify) Bone - Bone: dyn_tentacle_1"(ComponentPose=@n4, Alpha=@n3, Translation="-6,0,0", Rotation="-20,0,0", Scale="1,1,1") {
    rotationMode=BMM_Additive
    rotationSpace=BCS_BoneSpace
    translationMode=BMM_Additive
    translationSpace=BCS_BoneSpace
    alphaBoolBlend {blendOption=Linear}
    boneToModify {boneName=dyn_tentacle_1}
    rotation {pitch=-20}
    translation {x=-6}
   }
   n6 = Get Head Rot Dn() -> (Head Rot Dn:float)
   n7 = "Transform (Modify) Bone - Bone: dyn_tentacle_2"(ComponentPose=@n5, Alpha=@n6, Translation="-4,0,0", Rotation="-25,0,0", Scale="1,1,1") {
    rotationMode=BMM_Additive
    rotationSpace=BCS_BoneSpace
    translationMode=BMM_Additive
    translationSpace=BCS_BoneSpace
    alphaBoolBlend {blendOption=Linear}
    boneToModify {boneName=dyn_tentacle_2}
    rotation {pitch=-25}
    translation {x=-4}
   }
   n8 = "Transform (Modify) Bone - Bone: dyn_tentacle_1"(ComponentPose=@n7, Alpha=@n6, Translation="-6,0,0", Rotation="-20,0,0", Scale="1,1,1") {
    rotationMode=BMM_Additive
    rotationSpace=BCS_BoneSpace
    translationMode=BMM_Additive
    translationSpace=BCS_BoneSpace
    alphaBoolBlend {blendOption=Linear}
    boneToModify {boneName=dyn_tentacle_1}
    rotation {pitch=-20}
    translation {x=-6}
   }
   Output Pose(Result=Component To Local(ComponentPose=@n8)) {layerGroup=None}
  ]}
 EventGraph {nodes [
   n1 = Event BlueprintUpdateAnimation()
   > Set Delta Time X(Delta Time X=@n1.DeltaTimeX)
   > n2 = Sequence!(execute) -> (then_0, then_1, then_2:Exec, then_3:Exec)
   > n3 = Gravity Override State Function DEPRECATED(CharacterPartAnimInstance=Self-Reference(), GameTimeInSeconds=0, GravityOverrideParams=Get Gravity Override Params())
   > Set Gravity Override(Gravity Override=@n3)
   Set bodyMesh(execute=@n2.then_1, bodyMesh:Skeletal Mesh Component Object Reference=GetPartSkeletalMeshComponent()) -> (Output_Get:Skeletal Mesh Component Object Reference)
   Set head Rot Up(
    execute=@n2.then_2
    head Rot Up:float=GetAlphaFromReferenceTransformDelta(
     SkeletalMeshComponent=Get bodyMesh() -> (bodyMesh:Skeletal Mesh Component Object Reference)
     BoneName=head
     BaseName=neck_01
     InputRange=(X=-30,Y=0)
     OutputRange=(X=1,Y=0)
     TwistAxis="1,0,0"
    )
   ) -> (Output_Get:float)
   Set Head Rot Dn(
    execute=@n2.then_3
    Head Rot Dn:float=GetAlphaFromReferenceTransformDelta(SkeletalMeshComponent=Get bodyMesh(), BoneName=head, BaseName=neck_01, InputRange=(X=0,Y=20), TwistAxis="1,0,0")
   ) -> (Output_Get:float)
  ]}
]
variables [
 bodyMesh {type=SkeletalMeshComponent}
 Delta Time X {type=number}
 Gravity Override {type=Vector}
 Gravity Override Params {type=GravityOverrideParamsStruct}
 Head Rot Dn {type=number}
 head Rot Up {type=number}
]
